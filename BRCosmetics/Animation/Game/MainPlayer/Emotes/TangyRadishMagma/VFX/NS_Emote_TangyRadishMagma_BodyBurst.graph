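# Niagara one-shot burst VFX for the TangyRadishMagma emote: a brief fiery yellow-orange flash offset up to body height.
# A single Burst emitter fires four instantaneous spawn groups, each setting VisibilityTag, color, lifetime and size to feed three sprite renderers and one mesh renderer.
# Curves drive a fast pop-and-shrink in scale and alpha fade, with separate color and size attributes for mesh and corona layers.
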NiagaraSystem /BRCosmetics/Animation/Game/MainPlayer/Emotes/TangyRadishMagma/VFX/NS_Emote_TangyRadishMagma_BodyBurst
dependencies {
 usedDataInterfaces [
  &NiagaraDataInterfaceVectorCurve
  &NiagaraDataInterfaceCurve
  &NiagaraDataInterfaceVector2DCurve
  &NiagaraDataInterfaceColorCurve
 ]
 usedDynamicInputs []
 usedModules [&SystemState]
}
emitters {Burst {
  emitterUpdateScript [
   &EmitterState(Life Cycle Mode=::NewEnumerator1, Loop Behavior=::NewEnumerator1, Loop Duration=0.7, UseLoopDelay=false)
   &SpawnBurst_Instantaneous(Use Spawn Probability=false, UseLoopCountLimit=false) {module=SpawnBurst_Instantaneous002}
   &SpawnBurst_Instantaneous(Spawn Group=1, Use Spawn Probability=false, UseLoopCountLimit=false)
   &SpawnBurst_Instantaneous(Spawn Group=2, Use Spawn Probability=false, UseLoopCountLimit=false) {module=SpawnBurst_Instantaneous001}
   &SpawnBurst_Instantaneous(Spawn Group=3, Use Spawn Probability=false, UseLoopCountLimit=false) {module=SpawnBurst_Instantaneous003}
  ]
  particleSpawnScript [
   /SetVariables(Particles.VisibilityTag=&SpawnGroup()) {isSetParameters=true}
   &InitializeParticle(
    Color=&MergeHelper_ColorSwitch(
     0=<rgba 2 0.9352 0.2272 1.2>
     1=<rgba 1 0.1544 0 0.7>
     2=<rgba 2.5 0.9171 0 1.25>
     3=?
     Index=&SpawnGroup()
     Num=::NewEnumerator1
    )
    Lifetime=&MergeHelper_FloatSwitch(0=0.4, 1=0.4, 2=0.6, 3=0.5, Index=&SpawnGroup(), Num=::NewEnumerator2)
    Mesh Scale Mode=::NewEnumerator3
    Mesh Uniform Scale=1.2
    Position Offset=<0 0 115>
    Sprite Rotation Angle=90
    Sprite Rotation Mode=::NewEnumerator0
    Sprite Size=&MergeHelper_Vector2Switch(0=<80 130>, 1=<80 130>, 2=<0 0>, 3=<185 185>, 4=?, Index=&SpawnGroup(), Num=::NewEnumerator2)
    Sprite Size Mode=::NewEnumerator4
    UsePositionOffset=true
    Write Material Random=true
   )
   &CameraOffset(Camera Offset Amount=-50)
  ]
  particleUpdateScript [
   &ParticleState()
   &ScaleMeshSize(Scale Factor=&VectorFromCurve(
     VectorCurve=DataInterface() {propertyValues {
       XCurve {
        defaultValue=0
        keys [
         {arriveTangent=39.1075, interpMode=RCIM_Cubic, leaveTangent=39.1075, tangentMode=RCTM_User, time=0, value=0}
         {interpMode=RCIM_Cubic, tangentMode=RCTM_User, time=0.0605, value=1.0274}
         {interpMode=RCIM_Cubic, tangentMode=RCTM_User, time=0.1098, value=0.7268}
         {interpMode=RCIM_Cubic, tangentMode=RCTM_User, time=0.9849, value=1.05}
        ]
       }
       YCurve {keys [
         {arriveTangent=39.1075, interpMode=RCIM_Cubic, leaveTangent=39.1075, leaveTangentWeight=0.3333, tangentMode=RCTM_User, time=0, value=0}
         {arriveTangentWeight=0.0396, interpMode=RCIM_Cubic, leaveTangentWeight=0.0294, tangentMode=RCTM_User, time=0.0605, value=1.0274}
         {arriveTangentWeight=0.069, interpMode=RCIM_Cubic, leaveTangentWeight=0.2643, tangentMode=RCTM_User, time=0.1098, value=0.7268}
         {arriveTangentWeight=0.2556, interpMode=RCIM_Cubic, leaveTangentWeight=0.0087, tangentMode=RCTM_User, time=0.988, value=0}
        ]}
       ZCurve {
        defaultValue=0
        keys [
         {arriveTangent=39.1075, interpMode=RCIM_Cubic, leaveTangent=39.1075, tangentMode=RCTM_User, time=0, value=0}
         {arriveTangentWeight=0.0396, interpMode=RCIM_Cubic, leaveTangentWeight=0.0294, tangentMode=RCTM_User, time=0.0605, value=1.0274}
         {arriveTangentWeight=0.069, interpMode=RCIM_Cubic, tangentMode=RCTM_User, time=0.1098, value=0.7268}
         {arriveTangentWeight=0.2593, interpMode=RCIM_Cubic, tangentMode=RCTM_User, time=0.9849, value=0}
        ]
       }
      }}
     Scale Curve=<1.5 1.5 1.5>
    ))
   &ScaleSpriteSize(
    Non-Uniform Curve Scale=<1.5 1>
    Non-Uniform Curve Sprite Scale=DataInterface() {propertyValues {
      XCurve {keys [
        {arriveTangent=41.408, interpMode=RCIM_Cubic, leaveTangent=41.408, tangentMode=RCTM_User, time=0, value=0.1}
        {interpMode=RCIM_Cubic, tangentMode=RCTM_User, time=0.0457, value=0.9462}
        {arriveTangent=-11.5805, interpMode=RCIM_Cubic, leaveTangent=-0.9291, tangentMode=RCTM_Break, time=0.1042, value=0.4729}
        {arriveTangent=-0.1541, interpMode=RCIM_Cubic, leaveTangent=-0.1541, tangentMode=RCTM_User, time=1, value=0}
       ]}
      YCurve {keys [
        {interpMode=RCIM_Cubic, leaveTangent=4.0442, tangentMode=RCTM_Break, time=0, value=0.1}
        {arriveTangent=4.8062, interpMode=RCIM_Cubic, leaveTangent=4.8062, time=0.1027, value=0.4785}
        {interpMode=RCIM_Cubic, tangentMode=RCTM_User, time=0.1788, value=0.9593}
        {arriveTangent=-0.7572, interpMode=RCIM_Cubic, leaveTangent=-0.7572, tangentMode=RCTM_User, time=0.3212, value=0.2892}
        <at 0.9979 0.0156>
       ]}
     }}
    Scale Sprite Size Mode=::NewEnumerator3
   )
   &ScaleColor(
    Scale Alpha=&FloatFromCurve(FloatCurve=DataInterface() {propertyValues {Curve {keys [
         <at -0.0046 1.0113 RCIM_Cubic>
         {arriveTangent=-1.0036, interpMode=RCIM_Cubic, leaveTangent=-1.0036, time=0.5003, value=0.2765}
         <at 1.0011 0.002 RCIM_Cubic>
        ]}}})
    ScaleA=true
    ScaleRGB=false
   )
   &SolveForcesAndVelocity(Clamp Velocity=false, Limit Acceleration=false)
   /SetVariables(Particles.Mesh Scale Color:LinearColor=&ColorFromCurve(ColorCurve=DataInterface() {propertyValues {
       AlphaCurve {keys [<at 0.2748 0.9 RCIM_Cubic>,<at 0.5629 0.4>,<at 1 0 RCIM_Cubic>]}
       BlueCurve {keys [<at 0 0 RCIM_Cubic>,<at 0.1473 0>,<at 0.3448 0>,<at 0.7105 0 RCIM_Cubic>,<at 0.945 0>]}
       GreenCurve {keys [
         <at 0 0.1478 RCIM_Cubic>
         <at 0.1473 0.8211>
         <at 0.3448 0.6972>
         {arriveTangent=-1.1436, interpMode=RCIM_Cubic, leaveTangent=-1.1436, time=0.7105, value=0.1243}
         <at 0.945 0.0108>
        ]}
       RedCurve {keys [
         <at 0 1 RCIM_Cubic>
         <at 0.1473 2>
         <at 0.3448 2>
         {arriveTangent=-1.6661, interpMode=RCIM_Cubic, leaveTangent=-1.6661, time=0.7105, value=1}
         <at 0.945 1>
        ]}
      }})) {isSetParameters=true}
   /SetVariables(
    Particles.Corona Scale Color:LinearColor=&MakeLinearColorFromVectorAndFloat(
     "Vector (RGB)"=&MakeVectorFromLinearColorRGB(Color=<rgba 3 0.764 0 2>)
     "Float (Alpha)"=&FloatFromCurve(FloatCurve=DataInterface() {propertyValues {Curve {
         defaultValue=1
         keys [
          {arriveTangent=20.985, interpMode=RCIM_Cubic, leaveTangent=20.985, tangentMode=RCTM_User, time=0, value=0}
          {arriveTangent=-3.9091, interpMode=RCIM_Cubic, leaveTangent=-3.9091, tangentMode=RCTM_User, time=0.0828, value=1.004}
          {arriveTangent=-1.5525, interpMode=RCIM_Cubic, leaveTangent=-1.5525, tangentMode=RCTM_User, time=0.2526, value=0.4508}
          {arriveTangent=-0.0405, interpMode=RCIM_Cubic, leaveTangent=-0.0405, tangentMode=RCTM_User, time=1, value=0}
         ]
        }}})
    )
    Particles.Corona Scale Sprite Size:Vector2f=&Multiply_Vector2D(
     A=$Particles.Initial.SpriteSize
     B=&Vector2DFromFloat(Value=&FloatFromCurve(
       FloatCurve=DataInterface() {propertyValues {Curve {
          defaultValue=1
          keys [
           {arriveTangent=5.8075, interpMode=RCIM_Cubic, leaveTangent=5.8075, tangentMode=RCTM_User, time=0, value=0}
           {arriveTangent=1.1179, interpMode=RCIM_Cubic, leaveTangent=1.1179, tangentMode=RCTM_User, time=0.2465, value=0.8111}
           {arriveTangent=0.0224, interpMode=RCIM_Cubic, leaveTangent=0.0224, tangentMode=RCTM_User, time=1, value=1}
          ]
         }}}
       Scale Curve=0.73
      ))
    )
   ) {isSetParameters=true}
  ]
  rendererClasses [&NiagaraSpriteRendererProperties, &NiagaraMeshRendererProperties]
  renderers [
   {rendererClass=&NiagaraSpriteRendererProperties}
   {rendererClass=&NiagaraSpriteRendererProperties, rendererIndex=1}
   {rendererClass=&NiagaraMeshRendererProperties, rendererIndex=2}
   {rendererClass=&NiagaraSpriteRendererProperties, rendererIndex=3}
  ]
 }}
summary {
 systemName=NS_Emote_TangyRadishMagma_BodyBurst
 emitters [{emitterName=Burst, rendererClasses [&NiagaraSpriteRendererProperties, &NiagaraMeshRendererProperties]}]
 userVariables [
  User.bIsFrontEnd {type=NiagaraBool}
  User.bIsFrontEndPreview {type=NiagaraBool}
  User.RaysSpawn {type=NiagaraFloat, defaultValue=0, description=""}
 ]
}
