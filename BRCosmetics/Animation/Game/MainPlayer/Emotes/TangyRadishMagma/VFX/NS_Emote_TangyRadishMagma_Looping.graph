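# Looping Niagara effect for the TangyRadishMagma emote that adds fiery red-orange ambience, ember bursts, ground decals and pulsing glows.
# Ambient_Condensed packs six sprite layers into one emitter, using spawn groups to switch shape, size, color and fade.
# BurstEmbers001 fires timed, gravity-pulled spark bursts at preset offsets, scaled by User.Embers_Rate.
# Brightness, size and placement variously adapt to front-end preview, time of day, platform and quality level.
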NiagaraSystem /BRCosmetics/Animation/Game/MainPlayer/Emotes/TangyRadishMagma/VFX/NS_Emote_TangyRadishMagma_Looping
dependencies {
 usedDataInterfaces [
  &NiagaraDataInterfaceCurve
  &NiagaraDataInterfacePlatformSet
  &NiagaraDataInterfaceArrayFloat3
  &NiagaraDataInterfaceColorCurve
  &NiagaraDataInterfaceSpriteRendererInfo
  &NiagaraDataInterfaceVector2DCurve
  &NiagaraDataInterfaceArrayColor
  &NiagaraDataInterfaceCollisionQuery
  &NiagaraDataInterfaceLandscape
 ]
 usedDynamicInputs []
 usedModules [
  &SystemState
  &BlendSpaceHelper
  &self:SystemSpawnScript.SystemScriptSource.SystemScriptGraph.NiagaraNodeAssignment_0.SetVariables_#1
 ]
}
emitters {
 Ambient_Condensed {
  emitterUpdateScript [
   &EmitterState(Life Cycle Mode=::NewEnumerator1, Loop Duration=9.3, UseLoopDelay=false)
   &SpawnBurst_Instantaneous(Spawn Count=10, Use Spawn Probability=false, UseLoopCountLimit=false)
   &SpawnRate(SpawnRate=8, Use Spawn Probability=false)
   &SpawnRate(Spawn Probability=0.7, SpawnGroup=1, SpawnRate=3.5, Use Spawn Probability=true) {module=SpawnRate001}
   &SpawnRate(Spawn Probability=0.7, SpawnGroup=2, SpawnRate=3.5, Use Spawn Probability=true) {module=SpawnRate002}
   &SpawnBurst_Instantaneous(Spawn Count=3, Spawn Group=3, Use Spawn Probability=false, UseLoopCountLimit=false) {module=SpawnBurst_Instantaneous001}
   &SpawnRate(Spawn Probability=0.75, SpawnGroup=3, SpawnRate=2.2, Use Spawn Probability=true) {module=SpawnRate003}
   &SpawnBurst_Instantaneous(Spawn Count=3, Spawn Group=4, Use Spawn Probability=false, UseLoopCountLimit=true) {module=SpawnBurst_Instantaneous002}
   &SpawnRate(Spawn Probability=0.75, SpawnGroup=4, SpawnRate=1.8, Use Spawn Probability=true) {module=SpawnRate004}
   &SpawnBurst_Instantaneous(Spawn Group=5, Use Spawn Probability=false, UseLoopCountLimit=false) {module=SpawnBurst_Instantaneous003}
  ]
  particleSpawnScript [
   /SetVariables(Particles.VisibilityTag=&SpawnGroup()) {isSetParameters=true}
   &FN_Sprite_Attributes(Create Sprite Rotation Attribute=true, Write to Module Namespace=true)
   &InitializeParticle(
    Color=&MergeHelper_ColorSwitch(
     0=&RandomRangeLinearColor(
      Minimum=&RandomRangeLinearColor(Minimum=<rgba 2 0.1499 0 0.4>, Maximum=<rgba 3 0.1926 0 0.4>)
      Maximum=&RandomRangeLinearColor(Minimum=<rgba 4 0.3958 0.1665 0.4>, Maximum=<rgba 3 0.8427 0.1476 0.4>)
     )
     1=&MakeCustomLinearColorFromBool(Bool=$User.bIsFrontEndPreview, True Color=<rgba 1 0.0052 0.0052 0.24>, False Color=<rgba 1 0.0052 0.0052 0.2>)
     2=&MakeCustomLinearColorFromBool(Bool=$User.bIsFrontEndPreview, True Color=<rgba 1.15 0.7781 0.0949 0.27>, False Color=<rgba 1.15 0.5413 0.1458 0.25>)
     3=&RandomRangeLinearColor(Minimum=<rgba 1 0.0989 0.0798 0.8>, Maximum=<rgba 1 0.2178 0.0546 0.8>)
     4=&RandomRangeLinearColor(
      Minimum=&RandomRangeLinearColor(Minimum=<rgba 1 0.0989 0.0798 0.75>, Maximum=<rgba 1 0.0228 0 0.7>)
      Maximum=&RandomRangeLinearColor(Minimum=<rgba 1 0.2178 0.0546 0.7>, Maximum=<rgba 1 0.0315 0 0.8>)
     )
     5=&MakeCustomLinearColorFromBool(
      Bool=$User.bIsFrontEndPreview
      True Color=<rgba 1 0.0419 0 0.32>
      False Color=&MakeLinearColorFromVectorAndFloat(
       "Vector (RGB)"=&MakeVectorFromLinearColorRGB(Color=<rgba 1 0.1619 0 1>)
       "Float (Alpha)"=&Multiply_Float(
        A=&MakeCustomFloatFromBool(
         Bool=&SetBoolByIntComparison(A=$Engine.QualityLevel, B=0, Comparison Type=::NewEnumerator4)
         True Float=0.27
         False Float=1
        )
        B=&FNTimeOfDayScalar(Night Scale=0.15, Morning Scale=0.3, Day Scale=0.35, Evening Scale=0.3)
       )
      )
     )
     Index=&SpawnGroup()
     Num=::NewEnumerator4
    )
    Lifetime=&MergeHelper_FloatSwitch(
     0=&RandomRangeFloat(Minimum=0.556, Maximum=2)
     1=&RandomRangeFloat(Minimum=2.5, Maximum=3.5)
     2=&RandomRangeFloat(Minimum=2.5, Maximum=3.5)
     3=&RandomRangeFloat(Minimum=4, Maximum=9)
     4=&RandomRangeFloat(Minimum=4.5, Maximum=6)
     5=9.3
     Index=&SpawnGroup()
     Num=::NewEnumerator4
    )
    Position Offset=&MergeHelper_VectorSwitch(
     0=<0 0 -15>
     1=<0 0 72>
     2=&FN_FrontEndVector3Modifier(Gameplay=<0 0 150>, Front End=<0 0 110>)
     3=<0 0 0>
     4=<0 0 20>
     5=&MakeCustomVectorFromBool(
      Bool=&SetBoolByIntComparison(A=$Engine.QualityLevel, B=0, Comparison Type=::NewEnumerator4)
      True Vector=<0 0 110>
      False Vector=<0 0 70>
     )
     Index=&SpawnGroup()
     Num=::NewEnumerator4
    )
    Sprite Rotation Mode=::NewEnumerator2
    Sprite Size=&MergeHelper_Vector2Switch(
     0=&RandomRangeVector2D(Minimum=<26 150>, Maximum=<130 300>)
     1=&Vector2DFromFloat(Value=&RandomRangeFloat(Minimum=200, Maximum=300))
     2=&Vector2DFromFloat(Value=&RandomRangeFloat(Minimum=200, Maximum=300))
     3=&Vector2DFromFloat(Value=&RandomRangeFloat(Minimum=80, Maximum=190))
     4=&Vector2DFromFloat(Value=&RandomRangeFloat(Minimum=80, Maximum=190))
     5=&FN_FrontEndVector2Modifier(Front End:Vector2f=<320 400>, Gameplay:Vector2f=<400 500>)
     Index=&SpawnGroup()
     Num=::NewEnumerator4
    )
    Sprite Size Mode=::NewEnumerator4
    UsePositionOffset=true
    Write Material Random=true
   )
   &SubUVAnimation(
    Evaluation Type=::NewEnumerator0
    Output Alternate Renderer Binding=true
    "Sprite Renderer "=DataInterface() {propertyValues {SpriteRenderer=&self:Ambient_Condensed_0.NiagaraSpriteRendererProperties_2}}
    SubUV Animation Mode:ENiagara_SubUVLookupModeV2=::NewEnumerator2
    UseEndFrame=false
    UseStartFrame=false
   )
   &ShapeLocation(
    Hemisphere Distribution=<1 0.55>
    MaskSpawnGroup=false
    Non Uniform Scale=<1 1 0.35>
    Override Seed=false
    Sphere Radius=&MaskFloat(Mask Source=0, Mask Parameter=?, Mask=0, Masked Float=38, Passthrough Float=0)
    UseOwnerScale=true
   ) {module=ShapeLocation001}
   &ShapeLocation(
    Cylinder Height=&MaskFloat(
     Mask Source=0
     Mask Parameter=?
     Mask=1
     Masked Float=&FN_FrontEndFloatModifier(Gameplay=50, Front End=35)
     Passthrough Float=0
    )
    Cylinder Radius=&MaskFloat(
     Mask Source=0
     Mask Parameter=?
     Mask=1
     Masked Float=&FN_FrontEndFloatModifier(Gameplay=30, Front End=40)
     Passthrough Float=0
    )
    MaskSpawnGroup=false
    Non Uniform Scale=<1.5 1.25 1>
    Override Seed=false
    Shape Primitive=::NewEnumerator1
    Surface Only=false
    UseOwnerScale=true
   )
   &ShapeLocation(
    Cylinder Height=&MaskFloat(
     Mask Source=0
     Mask Parameter=?
     Mask=2
     Masked Float=&FN_FrontEndFloatModifier(Gameplay=85, Front End=35)
     Passthrough Float=0
    )
    Cylinder Radius=&MaskFloat(
     Mask Source=0
     Mask Parameter=?
     Mask=2
     Masked Float=&FN_FrontEndFloatModifier(Gameplay=30, Front End=40)
     Passthrough Float=0
    )
    MaskSpawnGroup=false
    Non Uniform Scale=<1.5 1.25 1>
    Override Seed=false
    Shape Primitive=::NewEnumerator1
    Surface Only=false
    UseOwnerScale=true
   ) {module=ShapeLocation002}
   &ShapeLocation(
    MaskSpawnGroup=false
    Non Uniform Scale=<1.25 1 1>
    Override Seed=false
    Ring Radius=&MaskFloat(Mask Source=0, Mask Parameter=?, Mask=3, Masked Float=42, Passthrough Float=0)
    Shape Primitive=::NewEnumerator4
    UseOwnerScale=true
   ) {module=ShapeLocation003}
   &ShapeLocation(
    Cylinder Height=&MaskFloat(Mask Source=0, Mask Parameter=?, Mask=4, Masked Float=25, Passthrough Float=0)
    Cylinder Radius=&MaskFloat(Mask Source=0, Mask Parameter=?, Mask=4, Masked Float=20, Passthrough Float=0)
    MaskSpawnGroup=false
    Non Uniform Scale=<1.25 1 1>
    Override Seed=false
    Shape Primitive=::NewEnumerator1
    Surface Only=false
    UseOwnerScale=true
   ) {module=ShapeLocation004}
   &AddVelocity(
    Module.Use Falloff Distance=false
    Origin Offset=<0 0 -50>
    Override Seed=false
    Velocity Mode=::NewEnumerator1
    Velocity Speed=&MaskFloat(Mask Source=0, Mask Parameter=?, Mask=0, Masked Float=0.01, Passthrough Float=0)
   )
   &AddVelocity(Velocity=&MergeHelper_VectorSwitch(
     0=<0 0 0.005>
     1=<0 0 0>
     2=<0 0 0>
     3=<0 0 3>
     4=<0 0 12>
     5=<0 0 0>
     Index=&SpawnGroup()
     Num=::NewEnumerator4
    )) {module=AddVelocity001}
   &CameraOffset(Camera Offset Amount=&MergeHelper_FloatSwitch(
     0=0
     1=-50
     2=-50
     3=&RandomRangeFloat(Minimum=-2, Maximum=25)
     4=&RandomRangeFloat(Minimum=-8, Maximum=15)
     5=-3
     Index=&SpawnGroup()
     Num=::NewEnumerator4
    ))
  ]
  particleUpdateScript [
   &ParticleState()
   &ScaleSpriteSize(
    Scale Factor=&MergeHelper_Vector2Switch(
     0=&Vector2DFromCurve(
      Vector2Curve=DataInterface() {propertyValues {
        XCurve {
         defaultValue=0
         keys [
          {interpMode=RCIM_Cubic, tangentMode=RCTM_User, time=0, value=0}
          {arriveTangentWeight=0.1667, interpMode=RCIM_Cubic, tangentMode=RCTM_User, time=0.5, value=1}
          {arriveTangent=-3.3018, interpMode=RCIM_Cubic, leaveTangent=-3.3018, tangentMode=RCTM_User, time=0.8652, value=0.4388}
          <at 1 0>
         ]
        }
        YCurve {
         defaultValue=0
         keys [
          {interpMode=RCIM_Cubic, tangentMode=RCTM_User, time=0, value=0}
          {arriveTangent=2.4744, interpMode=RCIM_Cubic, leaveTangent=2.4744, tangentMode=RCTM_User, time=0.209, value=0.758}
          {arriveTangentWeight=0.1667, interpMode=RCIM_Cubic, tangentMode=RCTM_User, time=0.5, value=1}
          <at 1 0>
         ]
        }
       }}
      ScaleCurve=<1 1>
     )
     1=&Vector2DFromFloat(Value=&FloatFromCurve(FloatCurve=DataInterface() {propertyValues {Curve {keys [{tangentMode=RCTM_User, time=0, value=1}, {tangentMode=RCTM_User, time=1, value=1.2}]}}}))
     2=&Vector2DFromFloat(Value=&FloatFromCurve(FloatCurve=DataInterface() {propertyValues {Curve {keys [{tangentMode=RCTM_User, time=0, value=1}, {tangentMode=RCTM_User, time=1, value=1.2}]}}}))
     3=<1 1>
     4=&Vector2DFromFloat(Value=&FloatFromCurve(FloatCurve=DataInterface() {propertyValues {Curve {
          defaultValue=0
          keys [
           {interpMode=RCIM_Cubic, tangentMode=RCTM_User, time=0, value=0}
           {interpMode=RCIM_Cubic, tangentMode=RCTM_User, time=0.29, value=1}
           {arriveTangent=-1.4085, interpMode=RCIM_Cubic, leaveTangent=-1.4085, time=0.6493, value=0.8525}
           {tangentMode=RCTM_User, time=1, value=0}
          ]
         }}}))
     5=<1 1>
     Index=&SpawnGroup()
     Num=::NewEnumerator4
    )
    Scale Sprite Size Mode=::NewEnumerator2
   )
   &SubUVAnimation(
    End Frame Range Override=40
    Output Alternate Renderer Binding=true
    "Sprite Renderer "=DataInterface() {propertyValues {SpriteRenderer=&self:Ambient_Condensed_0.NiagaraSpriteRendererProperties_1}}
    SubUV Lookup Index Scale=0.6
    UseEndFrame=true
    UseStartFrame=true
   ) {module=SubUVAnimation001}
   &SubUVAnimation(
    End Frame Range Override=40
    Output Alternate Renderer Binding=true
    "Sprite Renderer "=DataInterface() {propertyValues {SpriteRenderer=&self:Ambient_Condensed_0.NiagaraSpriteRendererProperties_3}}
    SubUV Lookup Index Scale=0.6
    UseEndFrame=true
    UseStartFrame=true
   ) {module=SubUVAnimation002}
   &SubUVAnimation(
    Output Alternate Renderer Binding=true
    "Sprite Renderer "=DataInterface() {propertyValues {SpriteRenderer=&self:Ambient_Condensed_0.NiagaraSpriteRendererProperties_0}}
    UseEndFrame=false
    UseStartFrame=false
   ) {module=SubUVAnimation003}
   &SubUVAnimation(
    Output Alternate Renderer Binding=true
    "Sprite Renderer "=DataInterface() {propertyValues {SpriteRenderer=&self:Ambient_Condensed_0.NiagaraSpriteRendererProperties_4}}
    UseEndFrame=false
    UseStartFrame=false
   ) {module=SubUVAnimation004}
   &GravityForce(Gravity=&MaskVector(
     Mask Source=0
     Mask Parameter=?
     Mask=4
     Masked Vector=&RandomRangeVector(Minimum=<0 0 8>, Maximum=<0 0 20>)
     Passthrough Vector=<0 0 0>
    ))
   &Drag(
    Drag=&MaskFloat(Mask Source=0, Mask Parameter=?, Mask=4, Masked Float=0.5, Passthrough Float=0)
    Rotational Drag:NiagaraFloat=&MaskFloat(Mask Source=0, Mask Parameter=?, Mask=4, Masked Float=1, Passthrough Float=0)
    Use Linear Drag=true
    Use Rotational Drag=true
   )
   &ScaleColor(
    Scale Alpha=&MergeHelper_FloatSwitch(
     0=&FloatFromCurve(
      FloatCurve=DataInterface() {propertyValues {Curve {
         defaultValue=0
         keys [
          {interpMode=RCIM_Cubic, tangentMode=RCTM_User, time=0, value=0}
          {interpMode=RCIM_Cubic, tangentMode=RCTM_User, time=0.5, value=1}
          {tangentMode=RCTM_User, time=1, value=0}
         ]
        }}}
      Scale Curve=&Multiply_Float(
       A=&MakeCustomFloatFromBool(
        Bool=&SetBoolByFloatComparison(
         A=&FN_PickFloatDependOnPlatform(
          TargetPlatform=DataInterface() {propertyValues {Platforms {
             qualityLevelMask=-30
             deviceProfileStates [
              {profileName=Windows, qualityLevelMask=0, setQualityLevelMask=2}
              {profileName=Mobile, qualityLevelMask=0, setQualityLevelMask=4294967295}
              {profileName=Mac, qualityLevelMask=0, setQualityLevelMask=2}
              {profileName=Linux, qualityLevelMask=0, setQualityLevelMask=2}
             ]
            }}}
          TargetPlatformActive=1
          TargetPlatformNoActive=0
         )
         B=0
        )
        True Float=0.5
        False Float=1
       )
       B=&Multiply_Float(
        A=&MakeCustomFloatFromBool(
         Bool=&SetBoolByIntComparison(A=$Engine.QualityLevel, B=0, Comparison Type=::NewEnumerator4)
         True Float=0.65
         False Float=1
        )
        B=&Multiply_Float(
         A=&RandomRangeFloat(Minimum=0.23, Maximum=1.1)
         B=&FNTimeOfDayScalar(Night Scale=0.4, Morning Scale=0.75, Evening Scale=0.75)
        )
       )
      )
     )
     1=&FloatFromCurve(
      FloatCurve=DataInterface() {propertyValues {Curve {keys [
          {interpMode=RCIM_Cubic, tangentMode=RCTM_User, time=0, value=0}
          {interpMode=RCIM_Cubic, tangentMode=RCTM_User, time=0.2766, value=1}
          {tangentMode=RCTM_User, time=1, value=0}
         ]}}}
      Scale Curve=&Multiply_Float(
       A=&MakeCustomFloatFromBool(
        Bool=&SetBoolByFloatComparison(
         A=&FN_PickFloatDependOnPlatform(
          TargetPlatform=DataInterface() {propertyValues {Platforms {
             qualityLevelMask=-30
             deviceProfileStates [
              {profileName=Windows, qualityLevelMask=0, setQualityLevelMask=2}
              {profileName=Mobile, qualityLevelMask=0, setQualityLevelMask=4294967295}
              {profileName=Mac, qualityLevelMask=0, setQualityLevelMask=2}
              {profileName=Linux, qualityLevelMask=0, setQualityLevelMask=2}
             ]
            }}}
          TargetPlatformActive=1
          TargetPlatformNoActive=0
         )
         B=0
        )
        True Float=0.5
        False Float=1
       )
       B=&Multiply_Float(
        A=&MakeCustomFloatFromBool(
         Bool=&SetBoolByIntComparison(A=$Engine.QualityLevel, B=0, Comparison Type=::NewEnumerator4)
         True Float=0.65
         False Float=1
        )
        B=&Multiply_Float(
         A=&FN_FrontEndFloatModifier(Gameplay=0.22, Front End=0.12)
         B=&FNTimeOfDayScalar(Night Scale=0.5, Morning Scale=0.75, Evening Scale=0.75)
        )
       )
      )
     )
     2=&FloatFromCurve(
      FloatCurve=DataInterface() {propertyValues {Curve {keys [
          {interpMode=RCIM_Cubic, tangentMode=RCTM_User, time=0, value=0}
          {interpMode=RCIM_Cubic, tangentMode=RCTM_User, time=0.2766, value=1}
          {tangentMode=RCTM_User, time=1, value=0}
         ]}}}
      Scale Curve=&Multiply_Float(
       A=&MakeCustomFloatFromBool(
        Bool=&SetBoolByFloatComparison(
         A=&FN_PickFloatDependOnPlatform(
          TargetPlatform=DataInterface() {propertyValues {Platforms {
             qualityLevelMask=-30
             deviceProfileStates [
              {profileName=Windows, qualityLevelMask=0, setQualityLevelMask=2}
              {profileName=Mobile, qualityLevelMask=0, setQualityLevelMask=4294967295}
              {profileName=Mac, qualityLevelMask=0, setQualityLevelMask=2}
              {profileName=Linux, qualityLevelMask=0, setQualityLevelMask=2}
             ]
            }}}
          TargetPlatformActive=1
          TargetPlatformNoActive=0
         )
         B=0
        )
        True Float=0.5
        False Float=1
       )
       B=&Multiply_Float(
        A=&MakeCustomFloatFromBool(
         Bool=&SetBoolByIntComparison(A=$Engine.QualityLevel, B=0, Comparison Type=::NewEnumerator4)
         True Float=0.65
         False Float=1
        )
        B=&Multiply_Float(
         A=&FN_FrontEndFloatModifier(Gameplay=0.3, Front End=0.12)
         B=&FNTimeOfDayScalar(Night Scale=0.5, Morning Scale=0.75, Evening Scale=0.75)
        )
       )
      )
     )
     3=&FloatFromCurve(
      FloatCurve=DataInterface() {propertyValues {Curve {
         defaultValue=1
         keys [
          {interpMode=RCIM_Cubic, tangentMode=RCTM_User, time=0, value=0}
          {interpMode=RCIM_Cubic, tangentMode=RCTM_User, time=0.2807, value=1.0588}
          {tangentMode=RCTM_User, time=1, value=0}
         ]
        }}}
      Scale Curve=&Multiply_Float(
       A=&MakeCustomFloatFromBool(
        Bool=&SetBoolByFloatComparison(
         A=&FN_PickFloatDependOnPlatform(
          TargetPlatform=DataInterface() {propertyValues {Platforms {
             qualityLevelMask=-29
             deviceProfileStates [
              {profileName=Windows, qualityLevelMask=0, setQualityLevelMask=3}
              {profileName=Linux, qualityLevelMask=0, setQualityLevelMask=3}
              {profileName=Mac, qualityLevelMask=0, setQualityLevelMask=3}
              {profileName=IOS, qualityLevelMask=0, setQualityLevelMask=4294967295}
             ]
            }}}
          TargetPlatformActive=1
          TargetPlatformNoActive=0
         )
         B=0
        )
        True Float=0.5
        False Float=1
       )
       B=&Multiply_Float(
        A=&MakeCustomFloatFromBool(
         Bool=&SetBoolByIntComparison(A=$Engine.QualityLevel, B=0, Comparison Type=::NewEnumerator4)
         True Float=0.65
         False Float=1
        )
        B=&Multiply_Float(
         A=&FN_FrontEndFloatModifier(Gameplay=0.65, Front End=0.3)
         B=&FNTimeOfDayScalar(Night Scale=0.6, Morning Scale=0.8, Evening Scale=0.8)
        )
       )
      )
     )
     4=&FloatFromCurve(
      FloatCurve=DataInterface() {propertyValues {Curve {
         defaultValue=0
         keys [
          {arriveTangent=13.0062, interpMode=RCIM_Cubic, leaveTangent=13.0062, tangentMode=RCTM_User, time=0, value=0}
          {interpMode=RCIM_Cubic, tangentMode=RCTM_User, time=0.1852, value=1}
          {arriveTangent=-0.7697, interpMode=RCIM_Cubic, leaveTangent=-0.7697, tangentMode=RCTM_User, time=0.7051, value=0.8679}
          {arriveTangent=-1.9584, interpMode=RCIM_Cubic, leaveTangent=-1.9584, tangentMode=RCTM_User, time=1, value=0}
         ]
        }}}
      Scale Curve=&Multiply_Float(
       A=&MakeCustomFloatFromBool(
        Bool=&SetBoolByFloatComparison(
         A=&FN_PickFloatDependOnPlatform(
          TargetPlatform=DataInterface() {propertyValues {Platforms {
             qualityLevelMask=-30
             deviceProfileStates [
              {profileName=Windows, qualityLevelMask=0, setQualityLevelMask=2}
              {profileName=Mobile, qualityLevelMask=0, setQualityLevelMask=4294967295}
              {profileName=Mac, qualityLevelMask=0, setQualityLevelMask=2}
              {profileName=Linux, qualityLevelMask=0, setQualityLevelMask=2}
             ]
            }}}
          TargetPlatformActive=1
          TargetPlatformNoActive=0
         )
         B=0
        )
        True Float=0.5
        False Float=1
       )
       B=&Multiply_Float(
        A=&MakeCustomFloatFromBool(
         Bool=&SetBoolByIntComparison(A=$Engine.QualityLevel, B=0, Comparison Type=::NewEnumerator4)
         True Float=0.33
         False Float=1
        )
        B=&Multiply_Float(
         A=&FN_FrontEndFloatModifier(
          Gameplay=&RandomRangeFloat(Minimum=0.6, Maximum=0.83)
          Front End=&FloatFromCurve(FloatCurve=DataInterface() {propertyValues {Curve {
              defaultValue=0.3
              keys [
               {interpMode=RCIM_Cubic, tangentMode=RCTM_User, time=0, value=0.4}
               {arriveTangent=-1.1537, interpMode=RCIM_Cubic, leaveTangent=-1.1537, tangentMode=RCTM_User, time=1, value=0}
              ]
             }}})
         )
         B=&FNTimeOfDayScalar(Night Scale=0.6, Morning Scale=0.8, Evening Scale=0.8)
        )
       )
      )
     )
     5=&FloatFromCurve(
      FloatCurve=DataInterface() {propertyValues {Curve {keys [<at 0 0>,<at 0.15 1>,<at 0.85 1>,<at 1 0>]}}}
      Scale Curve=&Multiply_Float(
       A=&MakeCustomFloatFromBool(
        Bool=&SetBoolByFloatComparison(
         A=&FN_PickFloatDependOnPlatform(
          TargetPlatform=DataInterface() {propertyValues {Platforms {
             qualityLevelMask=-29
             deviceProfileStates [
              {profileName=Windows, qualityLevelMask=0, setQualityLevelMask=3}
              {profileName=Linux, qualityLevelMask=0, setQualityLevelMask=3}
              {profileName=Mac, qualityLevelMask=0, setQualityLevelMask=3}
              {profileName=IOS, qualityLevelMask=0, setQualityLevelMask=4294967295}
             ]
            }}}
          TargetPlatformActive=1
          TargetPlatformNoActive=0
         )
         B=0
        )
        True Float=0.2
        False Float=1
       )
       B=&FN_FrontEndFloatModifier(Gameplay=1, Front End=0.6)
      )
     )
     Index=&SpawnGroup()
     Num=::NewEnumerator4
    )
    Scale RGB=&MergeHelper_VectorSwitch(
     0=<1 1 1>
     1=&Multiply_VectorByFloat(
      Vector=&MakeVectorFromLinearColorRGB(Color=<rgba 1.5 0.039 0 1>)
      Float=&RandomRangeFloat(Minimum=0.5, Maximum=1)
     )
     2=&Multiply_VectorByFloat(
      Vector=&MakeVectorFromLinearColorRGB(Color=<rgba 1 0.5243 0.0278 1>)
      Float=&RandomRangeFloat(Minimum=0.5, Maximum=1)
     )
     3=<1 1 1>
     4=<1 1 1>
     5=<1 1 1>
     Index=&SpawnGroup()
     Num=::NewEnumerator4
    )
    ScaleA=true
    ScaleRGB=true
   )
   &SolveForcesAndVelocity(Clamp Velocity=false, Limit Acceleration=false)
   &BlendSpace(Rotation Contribution=0)
  ]
  rendererClasses [&NiagaraSpriteRendererProperties]
  renderers [
   {rendererClass=&NiagaraSpriteRendererProperties}
   {rendererClass=&NiagaraSpriteRendererProperties, rendererIndex=1}
   {rendererClass=&NiagaraSpriteRendererProperties, rendererIndex=2}
   {rendererClass=&NiagaraSpriteRendererProperties, rendererIndex=3}
   {rendererClass=&NiagaraSpriteRendererProperties, rendererIndex=4}
   {rendererClass=&NiagaraSpriteRendererProperties, rendererIndex=5}
  ]
 }
 BurstEmbers001 {
  emitterUpdateScript [
   &EmitterState(Life Cycle Mode=::NewEnumerator1, Loop Delay=5.6, Loop Duration=3.2, UseLoopDelay=true)
   &SpawnRate(
    Spawn Probability=0.63
    SpawnGroup:NiagaraInt32=&RandomRangeInt(Minimum=0, Maximum=5, Evaluation Type=::NewEnumerator1, Randomness Mode=::NewEnumerator1)
    SpawnRate=&Multiply_Float(A=65, B=$User.Embers_Rate)
    Use Spawn Probability=true
   )
   &SpawnBurst_Instantaneous(Spawn Count=28, Spawn Probability=0.75, Spawn Time=0.3, Use Spawn Probability=false, UseLoopCountLimit=false)
   &SpawnBurst_Instantaneous(Spawn Count=15, Spawn Group=4, Spawn Probability=0.8, Spawn Time=0.44, Use Spawn Probability=false, UseLoopCountLimit=false) {module=SpawnBurst_Instantaneous005}
   &SpawnBurst_Instantaneous(Spawn Count=15, Spawn Group=3, Spawn Probability=0.8, Spawn Time=0.63, Use Spawn Probability=false, UseLoopCountLimit=false) {module=SpawnBurst_Instantaneous008}
   &SpawnBurst_Instantaneous(Spawn Count=22, Spawn Group=1, Spawn Probability=0.7, Spawn Time=1.05, Use Spawn Probability=false, UseLoopCountLimit=false) {module=SpawnBurst_Instantaneous009}
   &SpawnBurst_Instantaneous(Spawn Count=18, Spawn Group=4, Spawn Probability=0.75, Spawn Time=1.25, Use Spawn Probability=false, UseLoopCountLimit=false) {module=SpawnBurst_Instantaneous001}
   &SpawnBurst_Instantaneous(Spawn Count=28, Spawn Group=2, Spawn Probability=0.65, Spawn Time=2.1, Use Spawn Probability=false, UseLoopCountLimit=false) {module=SpawnBurst_Instantaneous002}
   &SpawnBurst_Instantaneous(Spawn Count=28, Spawn Group=1, Spawn Probability=0.8, Spawn Time=2.6, Use Spawn Probability=false, UseLoopCountLimit=false) {module=SpawnBurst_Instantaneous006}
   &SpawnBurst_Instantaneous(Spawn Count=22, Spawn Group=3, Spawn Probability=0.72, Spawn Time=1.8, Use Spawn Probability=false, UseLoopCountLimit=false) {module=SpawnBurst_Instantaneous003}
   &SpawnBurst_Instantaneous(Spawn Count=28, Spawn Probability=0.85, Spawn Time=0.5, Use Spawn Probability=false, UseLoopCountLimit=false) {module=SpawnBurst_Instantaneous007}
   &SpawnBurst_Instantaneous(Spawn Count=26, Spawn Group=4, Spawn Probability=0.7, Spawn Time=1.66, Use Spawn Probability=false, UseLoopCountLimit=false) {module=SpawnBurst_Instantaneous004}
   &SpawnBurst_Instantaneous(Spawn Count=24, Spawn Group=2, Spawn Probability=0.7, Spawn Time=0.81, Use Spawn Probability=false, UseLoopCountLimit=false) {module=SpawnBurst_Instantaneous010}
   &SpawnBurst_Instantaneous(Spawn Count=24, Spawn Probability=0.8, Spawn Time=2.2, Use Spawn Probability=false, UseLoopCountLimit=false) {module=SpawnBurst_Instantaneous011}
  ]
  particleSpawnScript [
   &PartitionParticles(Number Of Partitions=5, Particles Per Partition=5, Partition Mode:ENiagara_PartitionMode_v2=::NewEnumerator1)
   /SetVariables(Particles.VisibilityTag=&SpawnGroup()) {isSetParameters=true}
   &InitializeParticle(
    Color Maximum=<rgba 2 1.5055 1.0052 1.2>
    Color Minimum=<rgba 2 0.6152 0.1318 1.2>
    Color Mode=::NewEnumerator2
    Lifetime Max=&RandomRangeFloat(Minimum=0.35, Maximum=0.6)
    Lifetime Min=0.15
    Lifetime Mode=::NewEnumerator1
    Position Offset=<0 0 32>
    Sprite Size Max=<2 3.5>
    Sprite Size Min=<1 2.3>
    Sprite Size Mode=::NewEnumerator2
    UsePositionOffset=true
    Write Material Random=true
   )
   &ShapeLocation(
    MaskSpawnGroup=false
    Non Uniform Scale=&RandomRangeVector(Minimum=<1 2 1>, Maximum=<2 2 1>)
    Offset=&MergeHelper_VectorSwitch(
     0=&SelectVectorFromArray(
      Direct Array Index=$Particles.PartitionParticles.Partition
      Interpolation=&RandomRangeFloat(Minimum=0, Maximum=24)
      Normalized Interpolation=false
      Vector Selection Array=DataInterface() {propertyValues {FloatData [<-46 15 120>, <69 0 130>, <0 -20 170>, <23 0 110>, <-69 5 80>]}}
     )
     1=&SelectVectorFromArray(
      Direct Array Index=$Particles.PartitionParticles.Partition
      Interpolation=&RandomRangeFloat(Minimum=0, Maximum=24)
      Normalized Interpolation=false
      Vector Selection Array=DataInterface() {propertyValues {FloatData [<-69 0 40>, <23 30 130>, <0 26 60>, <-23 -10 140>, <92 -28 160>]}}
     )
     2=&SelectVectorFromArray(
      Direct Array Index=$Particles.PartitionParticles.Partition
      Interpolation=&RandomRangeFloat(Minimum=0, Maximum=24)
      Normalized Interpolation=false
      Vector Selection Array=DataInterface() {propertyValues {FloatData [<0 18 100>, <69 2 70>, <46 -30 165>, <-46 22 100>, <-23 40 50>]}}
     )
     3=&SelectVectorFromArray(
      Direct Array Index=$Particles.PartitionParticles.Partition
      Interpolation=&RandomRangeFloat(Minimum=0, Maximum=24)
      Normalized Interpolation=false
      Vector Selection Array=DataInterface() {propertyValues {FloatData [<-23 11 80>, <23 0 170>, <0 -30 100>, <-46 -17 160>, <-69 0 130>]}}
     )
     4=&SelectVectorFromArray(
      Direct Array Index=$Particles.PartitionParticles.Partition
      Interpolation=&RandomRangeFloat(Minimum=0, Maximum=24)
      Normalized Interpolation=false
      Vector Selection Array=DataInterface() {propertyValues {FloatData [<0 39 40>, <-69 0 80>, <46 -8 120>, <-92 15 100>, <69 -36 150>]}}
     )
     Index=&SpawnGroup()
     Num=::NewEnumerator3
    )
    Offset Mode=::NewEnumerator0
    Override Seed=false
    Sphere Radius=2
    UseOwnerScale=true
   )
   &AddVelocity(
    Module.Use Falloff Distance=false
    Override Seed=false
    Velocity Mode=::NewEnumerator1
    Velocity Origin=$Particles.ShapeLocation.ShapePosition
    Velocity Speed=&RandomRangeFloat(Minimum=60, Maximum=120)
   )
   &AddVelocity(Velocity=<0 0 -50>) {module=AddVelocity001}
   &CameraOffset(Camera Offset Amount=-10)
  ]
  particleUpdateScript [
   &ParticleState()
   &ScaleSpriteSize(
    Scale Sprite Size Mode=::NewEnumerator0
    Uniform Scale Factor=&FloatFromCurve(
     FloatCurve=DataInterface() {propertyValues {Curve {
        defaultValue=1
        keys [
         {arriveTangent=9.0725, interpMode=RCIM_Cubic, leaveTangent=9.0725, tangentMode=RCTM_User, time=0, value=1}
         <at 0.0126 1.8 RCIM_Cubic>
         {arriveTangent=-2.5249, interpMode=RCIM_Cubic, leaveTangent=-2.5249, tangentMode=RCTM_User, time=0.0386, value=1}
         {arriveTangent=-0.0405, interpMode=RCIM_Cubic, leaveTangent=-0.0405, tangentMode=RCTM_User, time=1, value=0}
        ]
       }}}
     Scale Curve=&RandomRangeFloat(Minimum=1, Maximum=1.4)
    )
   )
   &Drag(
    Drag=&Multiply_Float(
     A=&RandomRangeFloat(Minimum=3, Maximum=5)
     B=&FloatFromCurve(FloatCurve=DataInterface() {propertyValues {Curve {defaultValue=1, keys [{tangentMode=RCTM_User, time=0, value=0.5}, {tangentMode=RCTM_User, time=1, value=1}]}}})
    )
    Use Linear Drag=true
    Use Rotational Drag=true
   )
   &GravityForce(Gravity=&Multiply_Vector(
     A=<0 0 -980>
     B=&MakeVector(
      X=0
      Y=0
      Z=&FloatFromCurve(
       FloatCurve=DataInterface() {propertyValues {Curve {defaultValue=1, keys [{tangentMode=RCTM_User, time=0, value=0.5}, {tangentMode=RCTM_User, time=1, value=1}]}}}
       Scale Curve=&RandomRangeFloat(Minimum=0.7, Maximum=1)
      )
     )
    ))
   &SolveForcesAndVelocity(Clamp Velocity=false, Limit Acceleration=false)
   &FN_Sprite_Attributes(
    Color=&MakeLinearColorFromVectorAndFloat(
     "Vector (RGB)"=&MakeVectorFromLinearColorRGB(Color=&ColorFromCurve(ColorCurve=DataInterface() {propertyValues {
         AlphaCurve {keys [<at 0 1 RCIM_Cubic>,<at 0.5932 1>,<at 0.896 0>]}
         BlueCurve {keys [
           <at 0 0>
           {arriveTangent=0.3034, interpMode=RCIM_Cubic, leaveTangent=0.3034, time=0.076, value=0.2582}
           <at 0.4653 0.1412>
           <at 0.7393 0.0746>
           <at 0.8763 0 RCIM_Cubic>
          ]}
         GreenCurve {keys [
           <at 0 0.1202>
           {arriveTangent=0.899, interpMode=RCIM_Cubic, leaveTangent=0.899, time=0.076, value=1.2521}
           <at 0.4653 0.5385>
           <at 0.7393 0.1703>
           <at 0.8763 0.0304 RCIM_Cubic>
          ]}
         RedCurve {keys [<at 0 1>,<at 0.076 2 RCIM_Cubic>,<at 0.4653 1>,<at 0.7393 1>,<at 0.8763 0.375 RCIM_Cubic>]}
        }}))
     "Float (Alpha)"=&Multiply_Float(
      A=&FloatFromCurve(
       FloatCurve=DataInterface() {propertyValues {Curve {keys [
           {arriveTangent=0.2646, interpMode=RCIM_Cubic, leaveTangent=0.2646, tangentMode=RCTM_User, time=-0.0065, value=0}
           {arriveTangent=1.6267, interpMode=RCIM_Cubic, leaveTangent=1.6267, time=0.2689, value=0.2997}
           {arriveTangent=-0.4066, interpMode=RCIM_Cubic, leaveTangent=-0.4066, time=0.6083, value=1}
           {arriveTangent=-1.2325, interpMode=RCIM_Cubic, leaveTangent=-1.2325, tangentMode=RCTM_User, time=1.0011, value=0.002}
          ]}}}
       Scale Curve=0.4
      )
      B=$User.Embers_Rate
     )
    )
    Create Color Attribute=true
    Create SpriteSize Attribute=true
    SpriteSize=&RandomRangeVector2D(Minimum=<2 7>, Maximum=<5 9>)
    Write to Module Namespace=true
   )
   &ScaleSpriteSizeBySpeed(Max Scale Factor=<3 4.5>)
   &ScaleColor(
    Scale Alpha=&Multiply_Float(
     A=&FloatFromCurve(FloatCurve=DataInterface() {propertyValues {Curve {
         defaultValue=1
         keys [
          {arriveTangent=92.5681, interpMode=RCIM_Cubic, leaveTangent=92.5681, tangentMode=RCTM_User, time=0, value=0}
          {arriveTangent=-1.9705, interpMode=RCIM_Cubic, leaveTangent=-1.9705, tangentMode=RCTM_User, time=0.0646, value=1}
          {arriveTangent=-0.3047, interpMode=RCIM_Cubic, leaveTangent=-0.3047, tangentMode=RCTM_User, time=1, value=0}
         ]
        }}})
     B=&Multiply_Float(
      A=$User.Embers_Rate
      B=&FloatFromCurve(
       FloatCurve=DataInterface() {propertyValues {Curve {keys [<at 0 0.5>,<at 0.7037 1>,<at 0.997 0.5>]}}}
       CurveIndex=$Emitter.NormalizedLoopAge
       Scale Curve=1.3
      )
     )
    )
    ScaleA=true
    ScaleRGB=true
   )
  ]
  rendererClasses [&NiagaraSpriteRendererProperties]
  renderers [
   {rendererClass=&NiagaraSpriteRendererProperties}
   {rendererClass=&NiagaraSpriteRendererProperties, rendererIndex=1}
   {rendererClass=&NiagaraSpriteRendererProperties, rendererIndex=2}
   {rendererClass=&NiagaraSpriteRendererProperties, rendererIndex=3}
   {rendererClass=&NiagaraSpriteRendererProperties, rendererIndex=4}
   {rendererClass=&NiagaraSpriteRendererProperties, rendererIndex=5}
   {rendererClass=&NiagaraSpriteRendererProperties, rendererIndex=6}
   {rendererClass=&NiagaraSpriteRendererProperties, rendererIndex=7}
   {rendererClass=&NiagaraSpriteRendererProperties, rendererIndex=8}
   {rendererClass=&NiagaraSpriteRendererProperties, rendererIndex=9}
   {rendererClass=&NiagaraSpriteRendererProperties, rendererIndex=10}
   {rendererClass=&NiagaraSpriteRendererProperties, rendererIndex=11}
  ]
 }
 Decal {
  emitterUpdateScript [
   &EmitterState(Life Cycle Mode=::NewEnumerator1, Loop Duration=0.556, UseLoopDelay=false)
   &SpawnBurst_Instantaneous(Use Spawn Probability=false, UseLoopCountLimit=false)
  ]
  particleSpawnScript [
   &InitializeParticle(
    Color=&MakeCustomLinearColorFromBool(Bool=$User.bIsFrontEndPreview, True Color=<rgba 3 0.9975 0 1>, False Color=<rgba 3 0.8637 0.2567 1>)
    Lifetime=0.8
    UsePositionOffset=false
    Write Material Random=true
   )
   /SetVariables(Particles.DecalSize:Vector3f=<170 170 170>) {isSetParameters=true}
   /SetVariables(
    Particles.Glow DecalSize:Vector3f=<550 500 500>
    Particles.Glow Position:NiagaraPosition=&AddVectorToPosition(Position=/Niagara/DynamicInputs/Helpers/SimulationPosition(), Vector=<-15 -15 0>)
   ) {isSetParameters=true}
  ]
  particleUpdateScript [
   &ParticleState()
   &SolveForcesAndVelocity(Clamp Velocity=false, Limit Acceleration=false)
   &ScaleColor(
    Scale Alpha=&Multiply_Float(
     A=&FloatFromCurve(
      FloatCurve=DataInterface() {propertyValues {Curve {keys [
          {arriveTangent=-2.4342, interpMode=RCIM_Cubic, leaveTangent=-2.4342, tangentMode=RCTM_User, time=0, value=1}
          {arriveTangent=-0.3806, interpMode=RCIM_Cubic, leaveTangent=-0.3806, tangentMode=RCTM_User, time=1, value=0}
         ]}}}
      Scale Curve=&Multiply_Float(A=1, B=&FNTimeOfDayScalar(Night Scale=0.66, Morning Scale=0.8, Evening Scale=0.8))
     )
     B=$User.Fade
    )
    ScaleA=true
    ScaleRGB=true
   )
   &BlendSpace(Rotation Contribution=0)
   /SetVariables(Particles.Glow DecalColor:LinearColor=&MakeLinearColorFromVectorAndFloat(
     "Vector (RGB)"=&MakeVectorFromLinearColorRGB(Color=&MakeCustomLinearColorFromBool(Bool=$User.bIsFrontEndPreview, True Color=<rgba 2.5 0.0356 0 2>, False Color=<rgba 2 0.0285 0 2>))
     "Float (Alpha)"=&Multiply_Float(
      A=&FloatFromCurve(
       FloatCurve=DataInterface() {propertyValues {Curve {keys [
           {arriveTangent=-2.4342, interpMode=RCIM_Cubic, leaveTangent=-2.4342, tangentMode=RCTM_User, time=0, value=1}
           {arriveTangent=-0.3806, interpMode=RCIM_Cubic, leaveTangent=-0.3806, tangentMode=RCTM_User, time=1, value=0}
          ]}}}
       Scale Curve=&Multiply_Float(
        A=&FN_FrontEndFloatModifier(Gameplay=0.13, Front End=0.15)
        B=&FNTimeOfDayScalar(Night Scale=0.55, Morning Scale=0.8, Evening Scale=0.8)
       )
      )
      B=$User.Fade
     )
    )) {isSetParameters=true}
  ]
  rendererClasses [&NiagaraDecalRendererProperties]
  renderers [
   {rendererClass=&NiagaraDecalRendererProperties, rendererIndex=0}
   {rendererClass=&NiagaraDecalRendererProperties, rendererIndex=1}
  ]
 }
 Glows {
  emitterUpdateScript [
   &EmitterState(Life Cycle Mode=::NewEnumerator1, Loop Duration=2.224, UseLoopDelay=false)
   &SpawnBurst_Instantaneous(Use Spawn Probability=false, UseLoopCountLimit=false)
   &SpawnBurst_Instantaneous(Spawn Group=1, Use Spawn Probability=false, UseLoopCountLimit=false) {module=SpawnBurst_Instantaneous001}
  ]
  particleSpawnScript [
   /SetVariables(Particles.VisibilityTag=&SpawnGroup()) {isSetParameters=true}
   &InitializeParticle(
    Color=&MergeHelper_ColorSwitch(
     0=&MakeCustomLinearColorFromBool(Bool=$User.bIsFrontEndPreview, True Color=<rgba 3 1.2293 0 0.12>, False Color=<rgba 3 0.8755 0 0.2>)
     1=&MakeCustomLinearColorFromBool(Bool=$User.bIsFrontEndPreview, True Color=<rgba 3.5 0.0211 0 0.2>, False Color=<rgba 3.5 0.0211 0 0.12>)
     2=?
     3=?
     Index=&SpawnGroup()
     Num=::NewEnumerator0
    )
    Lifetime=3
    Position Offset=&MergeHelper_VectorSwitch(
     0=&FN_FrontEndVector3Modifier(Gameplay=<0 0 140>, Front End=<0 0 120>)
     1=<0 0 40>
     2=?
     3=?
     Index=&SpawnGroup()
     Num=::NewEnumerator0
    )
    Sprite Size=&MergeHelper_Vector2Switch(
     0=&FN_FrontEndVector2Modifier(Front End:Vector2f=<300 300>, Gameplay:Vector2f=<500 300>)
     1=&FN_FrontEndVector2Modifier(Front End:Vector2f=<480 300>, Gameplay:Vector2f=<425 270>)
     2=?
     3=?
     4=?
     Index=&SpawnGroup()
     Num=::NewEnumerator0
    )
    Sprite Size Mode=::NewEnumerator4
    UsePositionOffset=true
    Write Material Random=true
   )
   &CameraOffset(Camera Offset Amount=-2)
   /SetVariables(
    Particles.Glow DecalSize:Vector3f=<625 600 600>
    Particles.Glow Position:NiagaraPosition=&AddVectorToPosition(Position=&SimulationPosition(), Vector=<-15 -15 0>)
   ) {isSetParameters=true}
  ]
  particleUpdateScript [
   &ParticleState()
   &SolveForcesAndVelocity(Clamp Velocity=false, Limit Acceleration=false)
   &ScaleColor(
    Scale Alpha=&FloatFromCurve(
     FloatCurve=DataInterface() {propertyValues {Curve {
        defaultValue=1
        keys [
         {interpMode=RCIM_Cubic, tangentMode=RCTM_User, time=0, value=0}
         {arriveTangent=0.9928, interpMode=RCIM_Cubic, leaveTangent=0.9928, time=0.25, value=1}
         {interpMode=RCIM_Cubic, tangentMode=RCTM_User, time=0.5036, value=0.5}
         {arriveTangent=-1.0073, interpMode=RCIM_Cubic, leaveTangent=-1.0073, time=0.75, value=1}
         {tangentMode=RCTM_User, time=1, value=0}
        ]
       }}}
     Scale Curve=&Multiply_Float(
      A=&MakeCustomFloatFromBool(
       Bool=&SetBoolByFloatComparison(
        A=&FN_PickFloatDependOnPlatform(
         TargetPlatform=DataInterface() {propertyValues {Platforms {
            qualityLevelMask=-29
            deviceProfileStates [
             {profileName=Windows, qualityLevelMask=0, setQualityLevelMask=3}
             {profileName=Linux, qualityLevelMask=0, setQualityLevelMask=3}
             {profileName=Mac, qualityLevelMask=0, setQualityLevelMask=3}
             {profileName=IOS, qualityLevelMask=0, setQualityLevelMask=4294967295}
            ]
           }}}
         TargetPlatformActive=1
         TargetPlatformNoActive=0
        )
        B=0
       )
       True Float=0.25
       False Float=1
      )
      B=&Multiply_Float(
       A=&MakeCustomFloatFromBool(
        Bool=&SetBoolByIntComparison(A=$Engine.QualityLevel, B=0, Comparison Type=::NewEnumerator4)
        True Float=0.22
        False Float=1
       )
       B=&Multiply_Float(
        A=&FN_FrontEndFloatModifier(Gameplay=0.7, Front End=0.3)
        B=&FNTimeOfDayScalar(Night Scale=0.5, Morning Scale=0.8, Evening Scale=0.8)
       )
      )
     )
    )
    ScaleA=true
    ScaleRGB=true
   )
   /SetVariables(Particles.Glow DecalColor:LinearColor=&MakeLinearColorFromVectorAndFloat(
     "Vector (RGB)"=&MakeVectorFromLinearColorRGB(Color=&MakeCustomLinearColorFromBool(Bool=$User.bIsFrontEndPreview, True Color=<rgba 2.5 0.0649 0 2>, False Color=<rgba 2 0.0285 0 2>))
     "Float (Alpha)"=&Multiply_Float(
      A=&FloatFromCurve(
       FloatCurve=DataInterface() {propertyValues {Curve {
          defaultValue=1
          keys [
           {interpMode=RCIM_Cubic, tangentMode=RCTM_User, time=0, value=0}
           {arriveTangent=0.9928, interpMode=RCIM_Cubic, leaveTangent=0.9928, time=0.25, value=1}
           {interpMode=RCIM_Cubic, tangentMode=RCTM_User, time=0.5036, value=0.5}
           {arriveTangent=-1.0073, interpMode=RCIM_Cubic, leaveTangent=-1.0073, time=0.75, value=1}
           {tangentMode=RCTM_User, time=1, value=0}
          ]
         }}}
       Scale Curve=&Multiply_Float(
        A=&FN_FrontEndFloatModifier(Gameplay=0.2, Front End=0.13)
        B=&FNTimeOfDayScalar(Night Scale=0.5, Morning Scale=0.8, Evening Scale=0.8)
       )
      )
      B=$User.Fade
     )
    )) {isSetParameters=true}
  ]
  rendererClasses [&NiagaraSpriteRendererProperties, &NiagaraDecalRendererProperties]
  renderers [
   {rendererClass=&NiagaraSpriteRendererProperties}
   {rendererClass=&NiagaraSpriteRendererProperties, rendererIndex=1}
   {rendererClass=&NiagaraDecalRendererProperties, rendererIndex=2}
  ]
 }
 PrimaryLight {rendererClasses [], renderers []}
 PrimaryLight001 {rendererClasses [], renderers []}
 PrimaryLight002 {rendererClasses [], renderers []}
 RainEmbersCPU {
  emitterUpdateScript [
   &EmitterState(Life Cycle Mode=::NewEnumerator1, Loop Delay=5.27, Loop Duration=3.6, UseLoopDelay=true)
   &SpawnRate(
    Spawn Probability=0.7
    SpawnRate=&Multiply_Float(
     A=&FloatFromCurve(
      FloatCurve=DataInterface() {propertyValues {Curve {
         defaultValue=1
         keys [
          {tangentMode=RCTM_User, time=0, value=1}
          {arriveTangent=-0.0405, interpMode=RCIM_Cubic, leaveTangent=-0.0405, tangentMode=RCTM_User, time=0.794, value=1}
          {arriveTangent=-8.0089, interpMode=RCIM_Cubic, leaveTangent=-8.0089, tangentMode=RCTM_User, time=1, value=0}
         ]
        }}}
      CurveIndex=$Emitter.NormalizedLoopAge
      Scale Curve=400
     )
     B=$User.Embers_Rate
    )
    Use Spawn Probability=true
   )
   &SpawnRate(
    Spawn Probability=0.7
    SpawnGroup=1
    SpawnRate=&Multiply_Float(A=50, B=$User.Embers_Rate)
    Use Spawn Probability=true
   ) {module=SpawnRate001}
  ]
  particleSpawnScript [
   /SetVariables(Particles.VisibilityTag=&SpawnGroup()) {isSetParameters=true, module=SetVariables_#2}
   &PartitionParticles(Number Of Partitions=&MergeHelper_IntSwitch(0=9, 1=13, 2=?, 3=?, 4=?, 5=?, Index=&SpawnGroup(), Num=::NewEnumerator0))
   &InitializeParticle(
    Color Mode=::NewEnumerator0
    Lifetime=&MergeHelper_FloatSwitch(
     0=&RandomRangeFloat(Minimum=0.33, Maximum=0.88)
     1=&RandomRangeFloat(Minimum=0.2, Maximum=0.85)
     2=?
     3=?
     Index=&SpawnGroup()
     Num=::NewEnumerator0
    )
    Position Offset=<0 0 200>
    Sprite Rotation Mode=::NewEnumerator2
    Sprite Size Max=<2 3.5>
    Sprite Size Min=<1 2.3>
    Sprite Size Mode=::NewEnumerator2
    UsePositionOffset=true
    Write Material Random=true
   )
   &ShapeLocation(
    MaskSpawnGroup=false
    Non Uniform Scale=<1 15 1>
    Offset=&MaskVector(
     Mask Source=0
     Mask Parameter=?
     Mask=0
     Masked Vector=&SelectVectorFromArray(
      Direct Array Index=$Particles.PartitionParticles.Partition
      Vector Selection Array=DataInterface() {propertyValues {FloatData [<0 0 0>, <23 0 0>, <-23 0 0>, <46 0 0>, <-46 0 0>, <69 0 0>, <-69 0 0>, <92 0 0>, <-92 0 0>]}}
     )
     Passthrough Vector=<0 0 0>
    )
    Offset Mode=::NewEnumerator0
    Override Seed=false
    Sphere Radius=&MaskFloat(Mask Source=0, Mask Parameter=?, Mask=0, Masked Float=4.5, Passthrough Float=1)
    UseOwnerScale=true
   )
   &ShapeLocation(
    MaskSpawnGroup=false
    Non Uniform Scale=<1.15 19 1>
    Offset=&MaskVector(
     Mask Source=0
     Mask Parameter=?
     Mask=1
     Masked Vector=&SelectVectorFromArray(
      Direct Array Index=$Particles.PartitionParticles.Partition
      Vector Selection Array=DataInterface() {propertyValues {FloatData [
         <0 0 0>
         <11.5 0 0>
         <-11.5 0 0>
         <28.75 0 0>
         <-34.5 0 0>
         <57.5 0 0>
         <-52.5 0 0>
         <80.5 0 0>
         <-80.5 0 0>
         <38.5 0 0>
         <-60.5 0 0>
         <5.2 0 0>
         <-3.6 0 0>
        ]}}
      Random Seed=0
     )
     Passthrough Vector=<0 0 0>
    )
    Offset Mode=::NewEnumerator0
    Override Seed=false
    Sphere Radius=&MaskFloat(Mask Source=0, Mask Parameter=?, Mask=1, Masked Float=3, Passthrough Float=0)
    UseOwnerScale=true
   ) {module=ShapeLocation001}
   &CameraOffset(Camera Offset Amount=-10)
  ]
  particleUpdateScript [
   &ParticleState()
   &ScaleSpriteSize(
    Scale Sprite Size Mode=::NewEnumerator0
    Uniform Scale Factor=&Multiply_Float(
     A=&FloatFromCurve(FloatCurve=DataInterface() {propertyValues {Curve {
         defaultValue=0
         keys [
          {tangentMode=RCTM_User, time=0, value=1}
          {arriveTangent=-0.0405, interpMode=RCIM_Cubic, leaveTangent=-0.0405, tangentMode=RCTM_User, time=0.6709, value=1}
          {arriveTangent=-1.5658, interpMode=RCIM_Cubic, leaveTangent=-1.5658, tangentMode=RCTM_User, time=1, value=0}
         ]
        }}})
     B=&Waveform(
      "Waveform [1]"=::NewEnumerator0
      "[1] Frequency"=&RandomRangeFloat(Minimum=12, Maximum=24)
      "[1] Amplitude Min/Max"=<1 1.5>
     )
    )
   )
   &GravityForce()
   &WindForce(
    Wind Speed=&RandomRangeVector(Minimum=<-50 0 0>, Maximum=<50 0 0>)
    Wind Speed Scale=&FloatFromCurve(
     FloatCurve=DataInterface() {propertyValues {Curve {keys [
         {interpMode=RCIM_Cubic, tangentMode=RCTM_User, time=0, value=0}
         {interpMode=RCIM_Cubic, tangentMode=RCTM_User, time=0.25, value=1}
         {interpMode=RCIM_Cubic, tangentMode=RCTM_User, time=0.5, value=0}
         {interpMode=RCIM_Cubic, tangentMode=RCTM_User, time=0.75, value=1}
         {interpMode=RCIM_Cubic, tangentMode=RCTM_User, time=1, value=0}
        ]}}}
     Scale Curve=&RandomRangeFloat(Minimum=0, Maximum=7)
    )
   )
   &Drag(Drag=&RandomRangeFloat(Minimum=0.55, Maximum=0.8), Use Linear Drag=true, Use Rotational Drag=false)
   &ScaleColor(
    Scale Alpha=&Multiply_Float(
     A=&FloatFromCurve(FloatCurve=DataInterface() {propertyValues {Curve {keys [
          {arriveTangent=0.2646, interpMode=RCIM_Cubic, leaveTangent=0.2646, tangentMode=RCTM_User, time=-0.0065, value=0}
          {arriveTangent=1.6267, interpMode=RCIM_Cubic, leaveTangent=1.6267, time=0.2689, value=0.2997}
          {arriveTangent=-0.4066, interpMode=RCIM_Cubic, leaveTangent=-0.4066, time=0.6083, value=1}
          {arriveTangent=-1.2325, interpMode=RCIM_Cubic, leaveTangent=-1.2325, tangentMode=RCTM_User, time=1.0011, value=0.002}
         ]}}})
     B=$User.Embers_Rate
    )
    Scale RGB=&MakeVectorFromLinearColorRGB(Color=&MultiplyLinearColor(
      Linear Color=&SelectLinearColorFromArray(
       Array Sampling Mode=::NewEnumerator1
       Direct Array Index=$Particles.PartitionParticles.Partition
       Color Selection Array=DataInterface() {propertyValues {ColorData [
          <rgba 20 19.7969 17.5587 1>
          <rgba 18 17.7306 15.9477 1>
          <rgba 18 17.7306 15.9477 1>
          <rgba 8 5.8289 3.92 1>
          <rgba 8 5.8289 3.92 1>
          <rgba 4 1.6868 0.8352 1>
          <rgba 4 1.6868 0.8352 1>
          <rgba 2.8 0.9556 0.4432 1>
          <rgba 2.8 0.9556 0.4432 1>
         ]}}
       Random Seed=?
      )
      Scale Factor=&MakeVector4_FromLinearColor(LinearColor=&LerpLinearColors(
        StartColor=<rgba 1 1 1 1>
        EndColor=<rgba 1 0.152 0 1>
        LerpFactor=&FloatFromCurve(FloatCurve=DataInterface() {propertyValues {Curve {
            defaultValue=1
            keys [
             {interpMode=RCIM_Cubic, tangentMode=RCTM_User, time=0.3, value=0}
             {arriveTangent=5.384, interpMode=RCIM_Cubic, leaveTangent=5.384, tangentMode=RCTM_User, time=0.9742, value=1}
            ]
           }}})
       ))
     ))
    ScaleA=true
    ScaleRGB=true
   )
   &CurlNoiseForce(
    Bias Noise=false
    Noise Frequency=&RandomRangeFloat(Minimum=30, Maximum=120)
    Noise Strength=&Multiply_Float(
     A=&RandomRangeFloat(
      Minimum=&RandomRangeFloat(Minimum=0, Maximum=50)
      Maximum=&RandomRangeFloat(Minimum=250, Maximum=600)
     )
     B=&FloatFromCurve(
      FloatCurve=DataInterface() {propertyValues {Curve {
         defaultValue=1
         keys [{interpMode=RCIM_Cubic, tangentMode=RCTM_User, time=0, value=0.3}, {interpMode=RCIM_Cubic, tangentMode=RCTM_User, time=0.676, value=1}]
        }}}
      Scale Curve=&RandomRangeFloat(Minimum=0.8, Maximum=1.1)
     )
    )
    Pan Noise=false
   )
   &SolveForcesAndVelocity(Clamp Velocity=false, Limit Acceleration=false)
   &FN_Sprite_Attributes(
    Color=&MakeLinearColorFromVectorAndFloat(
     "Vector (RGB)"=&MakeVectorFromLinearColorRGB(Color=&ColorFromCurve(ColorCurve=DataInterface() {propertyValues {
         AlphaCurve {keys [<at 0 1 RCIM_Cubic>,<at 0.5932 1>,<at 0.896 0>]}
         BlueCurve {keys [
           <at 0 0>
           {arriveTangent=0.3034, interpMode=RCIM_Cubic, leaveTangent=0.3034, time=0.076, value=0.2582}
           <at 0.4653 0.1412>
           <at 0.7393 0.0746>
           <at 0.8763 0 RCIM_Cubic>
          ]}
         GreenCurve {keys [
           <at 0 0.1202>
           {arriveTangent=0.899, interpMode=RCIM_Cubic, leaveTangent=0.899, time=0.076, value=1.2521}
           <at 0.4653 0.5385>
           <at 0.7393 0.1703>
           <at 0.8763 0.0304 RCIM_Cubic>
          ]}
         RedCurve {keys [<at 0 1>,<at 0.076 2 RCIM_Cubic>,<at 0.4653 1>,<at 0.7393 1>,<at 0.8763 0.375 RCIM_Cubic>]}
        }}))
     "Float (Alpha)"=&Multiply_Float(
      A=&FloatFromCurve(
       FloatCurve=DataInterface() {propertyValues {Curve {keys [
           {arriveTangent=0.2646, interpMode=RCIM_Cubic, leaveTangent=0.2646, tangentMode=RCTM_User, time=-0.0065, value=0}
           {arriveTangent=1.6267, interpMode=RCIM_Cubic, leaveTangent=1.6267, time=0.2689, value=0.2997}
           {arriveTangent=-0.4066, interpMode=RCIM_Cubic, leaveTangent=-0.4066, time=0.6083, value=1}
           {arriveTangent=-1.2325, interpMode=RCIM_Cubic, leaveTangent=-1.2325, tangentMode=RCTM_User, time=1.0011, value=0.002}
          ]}}}
       Scale Curve=0.4
      )
      B=$User.Embers_Rate
     )
    )
    Create Color Attribute=true
    Create SpriteSize Attribute=true
    SpriteSize=&RandomRangeVector2D(Minimum=<5 11>, Maximum=<7 13>)
    Write to Module Namespace=true
   )
   &ScaleSpriteSizeBySpeed(Max Scale Factor=&RandomRangeVector2D(Minimum=<0.8 1>, Maximum=<1.2 3>))
  ]
  rendererClasses [&NiagaraSpriteRendererProperties]
  renderers [
   {rendererClass=&NiagaraSpriteRendererProperties}
   {rendererClass=&NiagaraSpriteRendererProperties, rendererIndex=1}
   {rendererClass=&NiagaraSpriteRendererProperties, rendererIndex=2}
   {rendererClass=&NiagaraSpriteRendererProperties, rendererIndex=3}
  ]
 }
 RainEmbersGPU {
  simTarget=GPUComputeSim
  emitterUpdateScript [
   &EmitterState(Life Cycle Mode=::NewEnumerator1, Loop Delay=5.27, Loop Duration=3.6, UseLoopDelay=true)
   &SpawnRate(
    Spawn Probability=0.7
    SpawnRate=&Multiply_Float(
     A=&FloatFromCurve(
      FloatCurve=DataInterface() {propertyValues {Curve {
         defaultValue=1
         keys [
          {tangentMode=RCTM_User, time=0, value=1}
          {arriveTangent=-0.0405, interpMode=RCIM_Cubic, leaveTangent=-0.0405, tangentMode=RCTM_User, time=0.794, value=1}
          {arriveTangent=-8.0089, interpMode=RCIM_Cubic, leaveTangent=-8.0089, tangentMode=RCTM_User, time=1, value=0}
         ]
        }}}
      CurveIndex=$Emitter.NormalizedLoopAge
      Scale Curve=400
     )
     B=$User.Embers_Rate
    )
    Use Spawn Probability=true
   )
   &SpawnRate(
    Spawn Probability=0.7
    SpawnGroup=1
    SpawnRate=&Multiply_Float(A=50, B=$User.Embers_Rate)
    Use Spawn Probability=true
   ) {module=SpawnRate001}
  ]
  particleSpawnScript [
   /SetVariables(Particles.VisibilityTag=&SpawnGroup()) {isSetParameters=true, module=SetVariables_#2}
   &PartitionParticles(Number Of Partitions=&MergeHelper_IntSwitch(0=9, 1=13, 2=?, 3=?, 4=?, 5=?, Index=&SpawnGroup(), Num=::NewEnumerator0))
   &InitializeParticle(
    Color Mode=::NewEnumerator0
    Lifetime=&MergeHelper_FloatSwitch(
     0=&RandomRangeFloat(Minimum=0.33, Maximum=0.88)
     1=&RandomRangeFloat(Minimum=0.2, Maximum=0.85)
     2=?
     3=?
     Index=&SpawnGroup()
     Num=::NewEnumerator0
    )
    Position Offset=<0 0 200>
    Sprite Rotation Mode=::NewEnumerator2
    Sprite Size Max=<2 3.5>
    Sprite Size Min=<1 2.3>
    Sprite Size Mode=::NewEnumerator2
    UsePositionOffset=true
    Write Material Random=true
   )
   &ShapeLocation(
    MaskSpawnGroup=false
    Non Uniform Scale=<1 15 1>
    Offset=&MaskVector(
     Mask Source=0
     Mask Parameter=?
     Mask=0
     Masked Vector=&SelectVectorFromArray(
      Direct Array Index=$Particles.PartitionParticles.Partition
      Vector Selection Array=DataInterface() {propertyValues {FloatData [<0 0 0>, <23 0 0>, <-23 0 0>, <46 0 0>, <-46 0 0>, <69 0 0>, <-69 0 0>, <92 0 0>, <-92 0 0>]}}
     )
     Passthrough Vector=<0 0 0>
    )
    Offset Mode=::NewEnumerator0
    Override Seed=false
    Sphere Radius=&MaskFloat(Mask Source=0, Mask Parameter=?, Mask=0, Masked Float=4.5, Passthrough Float=1)
    UseOwnerScale=true
   )
   &ShapeLocation(
    MaskSpawnGroup=false
    Non Uniform Scale=<1.15 19 1>
    Offset=&MaskVector(
     Mask Source=0
     Mask Parameter=?
     Mask=1
     Masked Vector=&SelectVectorFromArray(
      Direct Array Index=$Particles.PartitionParticles.Partition
      Vector Selection Array=DataInterface() {propertyValues {FloatData [
         <0 0 0>
         <11.5 0 0>
         <-11.5 0 0>
         <28.75 0 0>
         <-34.5 0 0>
         <57.5 0 0>
         <-52.5 0 0>
         <80.5 0 0>
         <-80.5 0 0>
         <38.5 0 0>
         <-60.5 0 0>
         <5.2 0 0>
         <-3.6 0 0>
        ]}}
      Random Seed=0
     )
     Passthrough Vector=<0 0 0>
    )
    Offset Mode=::NewEnumerator0
    Override Seed=false
    Sphere Radius=&MaskFloat(Mask Source=0, Mask Parameter=?, Mask=1, Masked Float=3, Passthrough Float=0)
    UseOwnerScale=true
   ) {module=ShapeLocation001}
   &CameraOffset(Camera Offset Amount=-10)
  ]
  particleUpdateScript [
   &ParticleState()
   &ScaleSpriteSize(
    Scale Sprite Size Mode=::NewEnumerator0
    Uniform Scale Factor=&Multiply_Float(
     A=&FloatFromCurve(FloatCurve=DataInterface() {propertyValues {Curve {
         defaultValue=0
         keys [
          {tangentMode=RCTM_User, time=0, value=1}
          {arriveTangent=-0.0405, interpMode=RCIM_Cubic, leaveTangent=-0.0405, tangentMode=RCTM_User, time=0.6709, value=1}
          {arriveTangent=-1.5658, interpMode=RCIM_Cubic, leaveTangent=-1.5658, tangentMode=RCTM_User, time=1, value=0}
         ]
        }}})
     B=&Waveform(
      "Waveform [1]"=::NewEnumerator0
      "[1] Frequency"=&RandomRangeFloat(Minimum=12, Maximum=24)
      "[1] Amplitude Min/Max"=<1 1.5>
     )
    )
   )
   &GravityForce()
   &WindForce(
    Wind Speed=&RandomRangeVector(Minimum=<-50 0 0>, Maximum=<50 0 0>)
    Wind Speed Scale=&FloatFromCurve(
     FloatCurve=DataInterface() {propertyValues {Curve {keys [
         {interpMode=RCIM_Cubic, tangentMode=RCTM_User, time=0, value=0}
         {interpMode=RCIM_Cubic, tangentMode=RCTM_User, time=0.25, value=1}
         {interpMode=RCIM_Cubic, tangentMode=RCTM_User, time=0.5, value=0}
         {interpMode=RCIM_Cubic, tangentMode=RCTM_User, time=0.75, value=1}
         {interpMode=RCIM_Cubic, tangentMode=RCTM_User, time=1, value=0}
        ]}}}
     Scale Curve=&RandomRangeFloat(Minimum=0, Maximum=7)
    )
   )
   &Drag(Drag=&RandomRangeFloat(Minimum=0.55, Maximum=0.8), Use Linear Drag=true, Use Rotational Drag=false)
   &ScaleColor(
    Scale Alpha=&Multiply_Float(
     A=&FloatFromCurve(FloatCurve=DataInterface() {propertyValues {Curve {keys [
          {arriveTangent=0.2646, interpMode=RCIM_Cubic, leaveTangent=0.2646, tangentMode=RCTM_User, time=-0.0065, value=0}
          {arriveTangent=1.6267, interpMode=RCIM_Cubic, leaveTangent=1.6267, time=0.2689, value=0.2997}
          {arriveTangent=-0.4066, interpMode=RCIM_Cubic, leaveTangent=-0.4066, time=0.6083, value=1}
          {arriveTangent=-1.2325, interpMode=RCIM_Cubic, leaveTangent=-1.2325, tangentMode=RCTM_User, time=1.0011, value=0.002}
         ]}}})
     B=$User.Embers_Rate
    )
    Scale RGB=&MakeVectorFromLinearColorRGB(Color=&MultiplyLinearColor(
      Linear Color=&SelectLinearColorFromArray(
       Array Sampling Mode=::NewEnumerator1
       Direct Array Index=$Particles.PartitionParticles.Partition
       Color Selection Array=DataInterface() {propertyValues {ColorData [
          <rgba 20 19.7969 17.5587 1>
          <rgba 18 17.7306 15.9477 1>
          <rgba 18 17.7306 15.9477 1>
          <rgba 8 5.8289 3.92 1>
          <rgba 8 5.8289 3.92 1>
          <rgba 4 1.6868 0.8352 1>
          <rgba 4 1.6868 0.8352 1>
          <rgba 2.8 0.9556 0.4432 1>
          <rgba 2.8 0.9556 0.4432 1>
         ]}}
       Random Seed=?
      )
      Scale Factor=&MakeVector4_FromLinearColor(LinearColor=&LerpLinearColors(
        StartColor=<rgba 1 1 1 1>
        EndColor=<rgba 1 0.152 0 1>
        LerpFactor=&FloatFromCurve(FloatCurve=DataInterface() {propertyValues {Curve {
            defaultValue=1
            keys [
             {interpMode=RCIM_Cubic, tangentMode=RCTM_User, time=0.3, value=0}
             {arriveTangent=5.384, interpMode=RCIM_Cubic, leaveTangent=5.384, tangentMode=RCTM_User, time=0.9742, value=1}
            ]
           }}})
       ))
     ))
    ScaleA=true
    ScaleRGB=true
   )
   &CurlNoiseForce(
    Bias Noise=false
    Noise Frequency=&RandomRangeFloat(Minimum=30, Maximum=120)
    Noise Strength=&Multiply_Float(
     A=&RandomRangeFloat(
      Minimum=&RandomRangeFloat(Minimum=0, Maximum=50)
      Maximum=&RandomRangeFloat(Minimum=250, Maximum=600)
     )
     B=&FloatFromCurve(
      FloatCurve=DataInterface() {propertyValues {Curve {
         defaultValue=1
         keys [{interpMode=RCIM_Cubic, tangentMode=RCTM_User, time=0, value=0.3}, {interpMode=RCIM_Cubic, tangentMode=RCTM_User, time=0.676, value=1}]
        }}}
      Scale Curve=&RandomRangeFloat(Minimum=0.8, Maximum=1.1)
     )
    )
    Pan Noise=false
   )
   &SolveForcesAndVelocity(Clamp Velocity=false, Limit Acceleration=false)
   &FN_Sprite_Attributes(
    Color=&MakeLinearColorFromVectorAndFloat(
     "Vector (RGB)"=&MakeVectorFromLinearColorRGB(Color=&ColorFromCurve(ColorCurve=DataInterface() {propertyValues {
         AlphaCurve {keys [<at 0 1 RCIM_Cubic>,<at 0.5932 1>,<at 0.896 0>]}
         BlueCurve {keys [
           <at 0 0>
           {arriveTangent=0.3034, interpMode=RCIM_Cubic, leaveTangent=0.3034, time=0.076, value=0.2582}
           <at 0.4653 0.1412>
           <at 0.7393 0.0746>
           <at 0.8763 0 RCIM_Cubic>
          ]}
         GreenCurve {keys [
           <at 0 0.1202>
           {arriveTangent=0.899, interpMode=RCIM_Cubic, leaveTangent=0.899, time=0.076, value=1.2521}
           <at 0.4653 0.5385>
           <at 0.7393 0.1703>
           <at 0.8763 0.0304 RCIM_Cubic>
          ]}
         RedCurve {keys [<at 0 1>,<at 0.076 2 RCIM_Cubic>,<at 0.4653 1>,<at 0.7393 1>,<at 0.8763 0.375 RCIM_Cubic>]}
        }}))
     "Float (Alpha)"=&Multiply_Float(
      A=&FloatFromCurve(
       FloatCurve=DataInterface() {propertyValues {Curve {keys [
           {arriveTangent=0.2646, interpMode=RCIM_Cubic, leaveTangent=0.2646, tangentMode=RCTM_User, time=-0.0065, value=0}
           {arriveTangent=1.6267, interpMode=RCIM_Cubic, leaveTangent=1.6267, time=0.2689, value=0.2997}
           {arriveTangent=-0.4066, interpMode=RCIM_Cubic, leaveTangent=-0.4066, time=0.6083, value=1}
           {arriveTangent=-1.2325, interpMode=RCIM_Cubic, leaveTangent=-1.2325, tangentMode=RCTM_User, time=1.0011, value=0.002}
          ]}}}
       Scale Curve=0.4
      )
      B=$User.Embers_Rate
     )
    )
    Create Color Attribute=true
    Create SpriteSize Attribute=true
    SpriteSize=&RandomRangeVector2D(Minimum=<5 11>, Maximum=<7 13>)
    Write to Module Namespace=true
   )
   &ScaleSpriteSizeBySpeed(Max Scale Factor=&RandomRangeVector2D(Minimum=<0.8 1>, Maximum=<1.2 3>))
  ]
  rendererClasses [&NiagaraSpriteRendererProperties]
  renderers [
   {rendererClass=&NiagaraSpriteRendererProperties}
   {rendererClass=&NiagaraSpriteRendererProperties, rendererIndex=1}
   {rendererClass=&NiagaraSpriteRendererProperties, rendererIndex=2}
   {rendererClass=&NiagaraSpriteRendererProperties, rendererIndex=3}
  ]
 }
}
summary {
 systemName=NS_Emote_TangyRadishMagma_Looping
 emitters [
  {emitterName=Glows, rendererClasses [&NiagaraSpriteRendererProperties, &NiagaraDecalRendererProperties]}
  {emitterName=PrimaryLight, rendererClasses []}
  {emitterName=Decal, rendererClasses [&NiagaraDecalRendererProperties]}
  {emitterName=BurstEmbers001, rendererClasses [&NiagaraSpriteRendererProperties]}
  {emitterName=PrimaryLight001, rendererClasses []}
  {emitterName=PrimaryLight002, rendererClasses []}
  {emitterName=Ambient_Condensed, rendererClasses [&NiagaraSpriteRendererProperties]}
  {emitterName=RainEmbersCPU, rendererClasses [&NiagaraSpriteRendererProperties]}
  {emitterName=RainEmbersGPU, simTarget=GPUComputeSim, rendererClasses [&NiagaraSpriteRendererProperties]}
 ]
 userVariables [
  User.bIsFrontEnd {type=NiagaraBool}
  User.bIsFrontEndPreview {type=NiagaraBool}
  User.Embers_Rate {type=NiagaraFloat, defaultValue=0, description=""}
  User.Fade {type=NiagaraFloat, defaultValue=0, description=""}
 ]
}
