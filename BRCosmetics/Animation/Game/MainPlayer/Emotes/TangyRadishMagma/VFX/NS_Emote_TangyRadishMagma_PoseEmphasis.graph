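# Niagara system for the TangyRadishMagma emote that emphasizes a pose with a short orange glow flash on the character's joints.
# One emitter bursts 16 sprites, using each particle's partition to pick its bone or socket index and glow size.
# User.IsLargeBody enlarges the glows, and time of day scales their color.
# Sprites follow the skeletal mesh and rotate while a curve snaps alpha up and fades it out.
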NiagaraSystem /BRCosmetics/Animation/Game/MainPlayer/Emotes/TangyRadishMagma/VFX/NS_Emote_TangyRadishMagma_PoseEmphasis
dependencies {
 usedDataInterfaces [
  &NiagaraDataInterfaceArrayFloat
  &NiagaraDataInterfaceArrayFloat2
  &NiagaraDataInterfaceArrayInt32
  &NiagaraDataInterfaceCurve
  &NiagaraDataInterfaceSkeletalMesh
 ]
 usedDynamicInputs [
  &Multiply_Float
  &MakeVector
  &SelectFloatFromArray
 ]
 usedModules [
  &self:SystemSpawnScript.SystemScriptSource.SystemScriptGraph.NiagaraNodeAssignment_0.SetVariables_#1
  &SystemState
 ]
}
emitters {Joint_Glows {
  emitterUpdateScript [
   &EmitterState(Inactive Response=::NewEnumerator1, Life Cycle Mode=::NewEnumerator1, Loop Behavior=::NewEnumerator1, Loop Duration=0.8, UseLoopDelay=false)
   &SpawnBurst_Instantaneous(Spawn Count=16, Use Spawn Probability=false, UseLoopCountLimit=false)
  ]
  particleSpawnScript [
   &PartitionParticles(Number Of Partitions=16)
   &InitializeParticle(Color Mode=::NewEnumerator0, Lifetime=0.8, UsePositionOffset=false, Write Material Random=true)
   &CameraOffset(Camera Offset Amount=10)
   /SetVariables(Particles.GlowScale:Vector2f=&Multiply_Vector2D(
     A=&SelectVector2DFromArray(
      Direct Array Index=$Particles.PartitionParticles.Partition
      Vector 2D Selection Array=DataInterface() {propertyValues {FloatData [
         <180 180>
         <180 180>
         <100 100>
         <100 100>
         <100 100>
         <100 100>
         <100 100>
         <100 100>
         <100 100>
         <130 130>
         <130 130>
         <130 130>
         <130 130>
         <100 100>
         <100 100>
         <140 140>
        ]}}
     )
     B=&MakeCustomVector2DFromBool(Bool=$User.IsLargeBody, True Vector2D=<1.2 1.2>, False Vector2D=<1 1>)
    )) {isSetParameters=true}
   /SetVariables(Particles.GlowCameraOffset=-10) {isSetParameters=true}
   /SetVariables(Particles.GlowColor=&Multiply_LinearColorByFloat(
     Linear Color=<rgba 1.25 0.3001 0 0.8>
     Float=&FNTimeOfDayScalar(Night Scale=0.5, Morning Scale=0.78, Evening Scale=0.78)
    )) {isSetParameters=true}
  ]
  particleUpdateScript [
   &ParticleState()
   &SpriteRotationRate(Rotation Rate=100)
   &SkeletalMeshLocation(
    Bone / Socket Index=&SelectIntFromArray(
     Direct Array Index=$Particles.PartitionParticles.Partition
     Int Selection Array=DataInterface() {propertyValues {IntData [0,1,2,3,4,5,6,7,8,9,10,11,12,13,14,15]}}
    )
    Bone Sampling Mode=::NewEnumerator4
    MaskSpawnGroup=false
    Orientation Sampling=::NewEnumerator1
    Skeletal Mesh=$System.Skeletal Mesh
   )
   &SolveForcesAndVelocity(Clamp Velocity=false, Limit Acceleration=false)
   &DynamicMaterialParameters(
    Index 0 Param 1=5
    Index 0 Param 2=1
    Index 0 Param 3=1
    Index 0 Param 4=1
    Module.Index0Param0Write=true
    Module.Index0Param1Write=true
    Module.Index0Param2Write=true
    Module.Index0Param3Write=true
   )
   /SetVariables(Particles.Scale Glow Color:LinearColor=&MakeLinearColorFromVectorAndFloat(
     "Vector (RGB)"=&MakeVectorFromLinearColorRGB(Color=$Particles.Initial.GlowColor)
     "Float (Alpha)"=&FloatFromCurve(FloatCurve=DataInterface() {propertyValues {Curve {
         defaultValue=1
         keys [
          {arriveTangent=66.9217, interpMode=RCIM_Cubic, leaveTangent=66.9217, tangentMode=RCTM_User, time=0, value=0}
          {interpMode=RCIM_Cubic, tangentMode=RCTM_User, time=0.0426, value=1}
          {arriveTangent=-1.0445, interpMode=RCIM_Cubic, leaveTangent=-1.0445, time=0.5373, value=0.2143}
          {interpMode=RCIM_Cubic, tangentMode=RCTM_User, time=1, value=0}
         ]
        }}})
    )) {isSetParameters=true}
  ]
  rendererClasses [&NiagaraSpriteRendererProperties]
  renderers [{rendererClass=&NiagaraSpriteRendererProperties}]
 }}
summary {
 systemName=NS_Emote_TangyRadishMagma_PoseEmphasis
 emitters [{emitterName=Joint_Glows, rendererClasses [&NiagaraSpriteRendererProperties]}]
 userVariables [User.IsLargeBody {type=NiagaraBool}]
}
