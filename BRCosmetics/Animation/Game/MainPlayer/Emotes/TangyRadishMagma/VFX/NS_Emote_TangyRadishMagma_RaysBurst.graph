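# Niagara VFX for the TangyRadishMagma emote: a burst of short-lived warm orange light-ray sprites from a single emitter.
# User.RaysSpawn gates eight single-particle bursts through two different comparisons against zero, splitting them into groups 0-2 and 3-7.
# Spawn group index switches each ray's color, offset, size, slight drift and alpha curve, and sets VisibilityTag alongside eight sprite renderers.
# Size curves pop each ray in quickly, then shrink it away.
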NiagaraSystem /BRCosmetics/Animation/Game/MainPlayer/Emotes/TangyRadishMagma/VFX/NS_Emote_TangyRadishMagma_RaysBurst
dependencies {
 usedDataInterfaces [
  &NiagaraDataInterfaceSpriteRendererInfo
  &NiagaraDataInterfaceCurve
  &NiagaraDataInterfaceVector2DCurve
 ]
 usedDynamicInputs []
 usedModules [&SystemState]
}
emitters {Light_Rays {
  emitterUpdateScript [
   &EmitterState(Life Cycle Mode=::NewEnumerator1, Loop Behavior=::NewEnumerator1, Loop Duration=1, UseLoopDelay=false)
   &SpawnBurst_Instantaneous(
    Spawn Count=&MakeCustomIntFromBool(Bool=&SetBoolByFloatComparison(A=$User.RaysSpawn, B=0), True INT=1, False INT=0)
    Use Spawn Probability=false
    UseLoopCountLimit=false
   )
   &SpawnBurst_Instantaneous(
    Spawn Count=&MakeCustomIntFromBool(Bool=&SetBoolByFloatComparison(A=$User.RaysSpawn, B=0), True INT=1, False INT=0)
    Spawn Group=1
    Use Spawn Probability=false
    UseLoopCountLimit=false
   ) {module=SpawnBurst_Instantaneous001}
   &SpawnBurst_Instantaneous(
    Spawn Count=&MakeCustomIntFromBool(Bool=&SetBoolByFloatComparison(A=$User.RaysSpawn, B=0), True INT=1, False INT=0)
    Spawn Group=2
    Use Spawn Probability=false
    UseLoopCountLimit=false
   ) {module=SpawnBurst_Instantaneous002}
   &SpawnBurst_Instantaneous(
    Spawn Count=&MakeCustomIntFromBool(
     Bool=&SetBoolByFloatComparison(A=$User.RaysSpawn, B=0, Comparison Type=::NewEnumerator6)
     True INT=1
     False INT=0
    )
    Spawn Group=3
    Use Spawn Probability=false
    UseLoopCountLimit=false
   ) {module=SpawnBurst_Instantaneous003}
   &SpawnBurst_Instantaneous(
    Spawn Count=&MakeCustomIntFromBool(
     Bool=&SetBoolByFloatComparison(A=$User.RaysSpawn, B=0, Comparison Type=::NewEnumerator6)
     True INT=1
     False INT=0
    )
    Spawn Group=4
    Use Spawn Probability=false
    UseLoopCountLimit=false
   ) {module=SpawnBurst_Instantaneous004}
   &SpawnBurst_Instantaneous(
    Spawn Count=&MakeCustomIntFromBool(
     Bool=&SetBoolByFloatComparison(A=$User.RaysSpawn, B=0, Comparison Type=::NewEnumerator6)
     True INT=1
     False INT=0
    )
    Spawn Group=5
    Use Spawn Probability=false
    UseLoopCountLimit=false
   ) {module=SpawnBurst_Instantaneous005}
   &SpawnBurst_Instantaneous(
    Spawn Count=&MakeCustomIntFromBool(
     Bool=&SetBoolByFloatComparison(A=$User.RaysSpawn, B=0, Comparison Type=::NewEnumerator6)
     True INT=1
     False INT=0
    )
    Spawn Group=6
    Use Spawn Probability=false
    UseLoopCountLimit=false
   ) {module=SpawnBurst_Instantaneous006}
   &SpawnBurst_Instantaneous(
    Spawn Count=&MakeCustomIntFromBool(
     Bool=&SetBoolByFloatComparison(A=$User.RaysSpawn, B=0, Comparison Type=::NewEnumerator6)
     True INT=1
     False INT=0
    )
    Spawn Group=7
    Use Spawn Probability=false
    UseLoopCountLimit=false
   ) {module=SpawnBurst_Instantaneous007}
  ]
  particleSpawnScript [
   /SetVariables(Particles.VisibilityTag=&SpawnGroup()) {isSetParameters=true}
   &InitializeParticle(
    Color=&MergeHelper_ColorSwitch(
     0=<rgba 2.7 0.5615 0.1208 0.3>
     1=<rgba 2.7 0.5615 0.1208 0.3>
     2=<rgba 2.7 0.5615 0.1208 0.3>
     3=<rgba 2.5 0.515 0.0827 0.28>
     4=<rgba 2.5 0.515 0.0827 0.28>
     5=<rgba 2.5 0.515 0.0827 0.28>
     6=<rgba 2.5 0.515 0.0827 0.28>
     7=<rgba 2.5 0.515 0.0827 0.28>
     Index=&SpawnGroup()
     Num=::NewEnumerator6
    )
    Lifetime=0.7
    Position Offset=&MergeHelper_VectorSwitch(
     0=<0 0 0>
     1=<0 -12 3>
     2=<0 8 -10>
     3=<0 0 30>
     4=<0 15 22>
     5=<0 -15 22>
     6=<0 26 22>
     7=<0 -26 22>
     Index=&SpawnGroup()
     Num=::NewEnumerator6
    )
    Position Offset Coordinate Space=::Local
    Sprite Size=&MergeHelper_Vector2Switch(
     0=&RandomRangeVector2D(Minimum=<130 290>, Maximum=<150 320>)
     1=&RandomRangeVector2D(Minimum=<30 100>, Maximum=<60 250>)
     2=&RandomRangeVector2D(Minimum=<60 90>, Maximum=<80 150>)
     3=<150 350>
     4=&RandomRangeVector2D(Minimum=<75 175>, Maximum=<100 200>)
     5=&RandomRangeVector2D(Minimum=<75 175>, Maximum=<100 200>)
     6=&RandomRangeVector2D(Minimum=<60 100>, Maximum=<75 150>)
     7=&RandomRangeVector2D(Minimum=<60 100>, Maximum=<75 150>)
     Index=&SpawnGroup()
     Num=::NewEnumerator6
    )
    Sprite Size Mode=::NewEnumerator4
    UsePositionOffset=true
    Write Material Random=true
   )
   &AddVelocity(Velocity=&MergeHelper_VectorSwitch(
     0=<0 -0.01 -0.01>
     1=<0 -0.02 -0.005>
     2=<0 -0.005 -0.015>
     3=<0 0 -0.01>
     4=<0 0.005 -0.01>
     5=<0 -0.005 -0.01>
     6=<0 0.015 -0.01>
     7=<0 -0.015 -0.01>
     Index=&SpawnGroup()
     Num=::NewEnumerator6
    ))
   &SubUVAnimation(
    "Sprite Renderer "=DataInterface() {propertyValues {SpriteRenderer=&self:Light_Rays_0.NiagaraSpriteRendererProperties_2}}
    SubUV Animation Mode:ENiagara_SubUVLookupModeV2=::NewEnumerator2
    UseEndFrame=false
    UseStartFrame=false
   )
  ]
  particleUpdateScript [
   &ParticleState()
   &ScaleSpriteSize(
    Non-Uniform Curve Sprite Scale=DataInterface() {propertyValues {
      XCurve {
       defaultValue=1
       keys [
        {interpMode=RCIM_Cubic, leaveTangentWeight=0.0327, tangentMode=RCTM_User, time=0.0019, value=0}
        {interpMode=RCIM_Cubic, tangentMode=RCTM_User, time=0.1, value=1}
        {
         arriveTangent=-1.7662
         arriveTangentWeight=0.1706
         interpMode=RCIM_Cubic
         leaveTangent=-1.7662
         leaveTangentWeight=0.1294
         tangentMode=RCTM_User
         time=0.6117
         value=0.6939
        }
        {arriveTangentWeight=0.3, interpMode=RCIM_Cubic, tangentMode=RCTM_User, time=1, value=0}
       ]
      }
      YCurve {keys [
        {interpMode=RCIM_Cubic, tangentMode=RCTM_User, time=0.0019, value=0}
        {interpMode=RCIM_Cubic, tangentMode=RCTM_User, time=0.1, value=1}
        {arriveTangent=-1.7662, interpMode=RCIM_Cubic, leaveTangent=-1.7662, tangentMode=RCTM_User, time=0.6117, value=0.6939}
        {interpMode=RCIM_Cubic, tangentMode=RCTM_User, time=1, value=0}
       ]}
     }}
    Scale Sprite Size Mode=::NewEnumerator3
   )
   &SolveForcesAndVelocity(Clamp Velocity=false, Limit Acceleration=false)
   &ScaleColor(
    Scale Alpha=&MergeHelper_FloatSwitch(
     0=&FloatFromCurve(FloatCurve=DataInterface() {propertyValues {Curve {
         defaultValue=1
         keys [
          {arriveTangent=9.0725, interpMode=RCIM_Cubic, leaveTangent=9.0725, tangentMode=RCTM_User, time=0, value=0}
          {arriveTangent=-2.5249, interpMode=RCIM_Cubic, leaveTangent=-2.5249, tangentMode=RCTM_User, time=0.1408, value=0.9847}
          {arriveTangent=-0.0405, interpMode=RCIM_Cubic, leaveTangent=-0.0405, tangentMode=RCTM_User, time=1, value=0}
         ]
        }}})
     1=&FloatFromCurve(FloatCurve=DataInterface() {propertyValues {Curve {
         defaultValue=1
         keys [
          {arriveTangent=9.0725, interpMode=RCIM_Cubic, leaveTangent=9.0725, tangentMode=RCTM_User, time=0, value=0}
          {arriveTangent=-2.5249, interpMode=RCIM_Cubic, leaveTangent=-2.5249, tangentMode=RCTM_User, time=0.1408, value=0.9847}
          {arriveTangent=-0.0405, interpMode=RCIM_Cubic, leaveTangent=-0.0405, tangentMode=RCTM_User, time=1, value=0}
         ]
        }}})
     2=&FloatFromCurve(FloatCurve=DataInterface() {propertyValues {Curve {
         defaultValue=1
         keys [
          {arriveTangent=9.0725, interpMode=RCIM_Cubic, leaveTangent=9.0725, tangentMode=RCTM_User, time=0, value=0}
          {arriveTangent=-2.5249, interpMode=RCIM_Cubic, leaveTangent=-2.5249, tangentMode=RCTM_User, time=0.1408, value=0.9847}
          {arriveTangent=-0.0405, interpMode=RCIM_Cubic, leaveTangent=-0.0405, tangentMode=RCTM_User, time=1, value=0}
         ]
        }}})
     3=&FloatFromCurve(FloatCurve=DataInterface() {propertyValues {Curve {
         defaultValue=1
         keys [
          {interpMode=RCIM_Cubic, leaveTangentWeight=0.349, tangentMode=RCTM_User, time=0, value=0}
          {arriveTangent=2.0569, interpMode=RCIM_Cubic, leaveTangent=-2.2075, tangentMode=RCTM_Break, time=0.1102, value=1}
          {arriveTangent=-1.1238, interpMode=RCIM_Cubic, leaveTangent=-1.1238, time=0.3929, value=0.2709}
          {arriveTangent=0.0514, interpMode=RCIM_Cubic, leaveTangent=0.0514, tangentMode=RCTM_User, time=1, value=0}
         ]
        }}})
     4=&FloatFromCurve(FloatCurve=DataInterface() {propertyValues {Curve {
         defaultValue=1
         keys [
          {interpMode=RCIM_Cubic, leaveTangentWeight=0.349, tangentMode=RCTM_User, time=0, value=0}
          {arriveTangent=2.0569, interpMode=RCIM_Cubic, leaveTangent=-2.2075, tangentMode=RCTM_Break, time=0.1102, value=1}
          {arriveTangent=-1.1238, interpMode=RCIM_Cubic, leaveTangent=-1.1238, time=0.3929, value=0.2709}
          {arriveTangent=0.0514, interpMode=RCIM_Cubic, leaveTangent=0.0514, tangentMode=RCTM_User, time=1, value=0}
         ]
        }}})
     5=&FloatFromCurve(FloatCurve=DataInterface() {propertyValues {Curve {
         defaultValue=1
         keys [
          {interpMode=RCIM_Cubic, leaveTangentWeight=0.349, tangentMode=RCTM_User, time=0, value=0}
          {arriveTangent=2.0569, interpMode=RCIM_Cubic, leaveTangent=-2.2075, tangentMode=RCTM_Break, time=0.1102, value=1}
          {arriveTangent=-1.1238, interpMode=RCIM_Cubic, leaveTangent=-1.1238, time=0.3929, value=0.2709}
          {arriveTangent=0.0514, interpMode=RCIM_Cubic, leaveTangent=0.0514, tangentMode=RCTM_User, time=1, value=0}
         ]
        }}})
     6=&FloatFromCurve(FloatCurve=DataInterface() {propertyValues {Curve {
         defaultValue=1
         keys [
          {interpMode=RCIM_Cubic, leaveTangentWeight=0.349, tangentMode=RCTM_User, time=0, value=0}
          {arriveTangent=2.0569, interpMode=RCIM_Cubic, leaveTangent=-2.2075, tangentMode=RCTM_Break, time=0.1102, value=1}
          {arriveTangent=-1.1238, interpMode=RCIM_Cubic, leaveTangent=-1.1238, time=0.3929, value=0.2709}
          {arriveTangent=0.0514, interpMode=RCIM_Cubic, leaveTangent=0.0514, tangentMode=RCTM_User, time=1, value=0}
         ]
        }}})
     7=&FloatFromCurve(FloatCurve=DataInterface() {propertyValues {Curve {
         defaultValue=1
         keys [
          {interpMode=RCIM_Cubic, leaveTangentWeight=0.349, tangentMode=RCTM_User, time=0, value=0}
          {arriveTangent=2.0569, interpMode=RCIM_Cubic, leaveTangent=-2.2075, tangentMode=RCTM_Break, time=0.1102, value=1}
          {arriveTangent=-1.1238, interpMode=RCIM_Cubic, leaveTangent=-1.1238, time=0.3929, value=0.2709}
          {arriveTangent=0.0514, interpMode=RCIM_Cubic, leaveTangent=0.0514, tangentMode=RCTM_User, time=1, value=0}
         ]
        }}})
     Index=&SpawnGroup()
     Num=::NewEnumerator6
    )
    ScaleA=true
    ScaleRGB=true
   )
  ]
  rendererClasses [&NiagaraSpriteRendererProperties]
  renderers [
   {rendererClass=&NiagaraSpriteRendererProperties}
   {rendererClass=&NiagaraSpriteRendererProperties, rendererIndex=1}
   {rendererClass=&NiagaraSpriteRendererProperties, rendererIndex=2}
   {rendererClass=&NiagaraSpriteRendererProperties, rendererIndex=3}
   {rendererClass=&NiagaraSpriteRendererProperties, rendererIndex=4}
   {rendererClass=&NiagaraSpriteRendererProperties, rendererIndex=5}
   {rendererClass=&NiagaraSpriteRendererProperties, rendererIndex=6}
   {rendererClass=&NiagaraSpriteRendererProperties, rendererIndex=7}
  ]
 }}
summary {
 systemName=NS_Emote_TangyRadishMagma_RaysBurst
 emitters [{emitterName=Light_Rays, rendererClasses [&NiagaraSpriteRendererProperties]}]
 userVariables [User.RaysSpawn {type=NiagaraFloat, defaultValue=0, description=""}]
}
